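# Physics asset for the cape on the SolveSolid medium male player cosmetic body.
# Four capsule chains hang from small anchors locked to the clavicles and upper spine, with swing limits widening toward the hem and zero twist.
# Angularly free cross-constraints tie neighboring chains together at an upper and a lower row.
# Capsules on torso, arms and legs, joined by angle-locked constraints, plus a large box below the root, provide body and floor shapes.
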PhysicsAsset /BRCosmetics/Characters/Player/Male/Medium/Bodies/M_MED_SolveSolid/Meshes/M_MED_SolveSolid_Cape_Physics
bodies {
 Dyn_cape_01_l1 [{shapeType=Capsule, length=0.505, radius=0.505, shapeName=Dyn_cape_01_l1_capsule, rotation=<rot 0 -90 0>}]
 Dyn_cape_01_r1 [{shapeType=Capsule, length=0.505, radius=0.505, shapeName=Dyn_cape_01_r1_capsule, rotation=<rot 0 -90 0>}]
 Dyn_cape_02_l [{shapeType=Capsule, length=10, radius=9, shapeName=Dyn_cape_02_l_capsule, center=<9 0 10.4477>, rotation=<rot 90 0 0>}]
 Dyn_cape_02_r [{shapeType=Capsule, length=10, radius=9, shapeName=Dyn_cape_02_r_capsule, center=<9 0 10.4477>, rotation=<rot 90 0 0>}]
 Dyn_cape_03_l [{shapeType=Capsule, length=12, radius=12, shapeName=Dyn_cape_03_l_capsule, center=<12 0 11.0477>, rotation=<rot 90 0 0>}]
 Dyn_cape_03_r [{shapeType=Capsule, length=12, radius=12, shapeName=Dyn_cape_03_r_capsule, center=<12 0 11.0477>, rotation=<rot 90 0 0>}]
 Dyn_cape_04_l [
  {shapeType=Capsule, length=21.5, radius=8.5, shapeName=Dyn_cape_04_l_capsule, center=<9 0 7>, rotation=<rot 90 -90 0>}
  {shapeType=Capsule, length=21.5, radius=8.5, shapeName=Dyn_cape_04_l_capsule, center=<20.6828 0 7>, rotation=<rot 90 -90 0>}
 ]
 Dyn_cape_04_r [
  {shapeType=Capsule, length=21.5, radius=8.5, shapeName=Dyn_cape_04_r_capsule, center=<9 0 7>, rotation=<rot 90 -90 0>}
  {shapeType=Capsule, length=21.5, radius=8.5, shapeName=Dyn_cape_04_r_capsule, center=<21.1972 0 7>, rotation=<rot 90 -90 0>}
 ]
 Dyn_cape_05_l [{shapeType=Capsule, length=21.5, radius=9.2, shapeName=Dyn_cape_05_l_capsule, center=<9 0 6.5>, rotation=<rot -0.0003 0 -85>}]
 Dyn_cape_05_r [{shapeType=Capsule, length=21.5, radius=9.2, shapeName=Dyn_cape_05_r_capsule, center=<9 0 6.5>, rotation=<rot -0.0003 0 -85>}]
 Dyn_cape_06_l [{shapeType=Capsule, length=21.5, radius=9.2, shapeName=Dyn_cape_06_l_capsule, center=<9 0 5.2>, rotation=<rot -0.0003 0 -85>}]
 Dyn_cape_06_r [{shapeType=Capsule, length=21.5, radius=9.2, shapeName=Dyn_cape_06_r_capsule, center=<9 0 5.2>, rotation=<rot -0.0003 0 -85>}]
 Dyn_cape_mid_01_l [{shapeType=Capsule, length=10, radius=9, shapeName=Dyn_cape_mid_01_l_capsule, center=<9 0 10.4477>, rotation=<rot 90 0 0>}]
 Dyn_cape_mid_01_l_MSTR [{shapeType=Capsule, length=0.505, radius=0.505, shapeName=Dyn_cape_mid_01_l_MSTR_capsule, rotation=<rot 0 -90 0>}]
 Dyn_cape_mid_01_r [{shapeType=Capsule, length=10, radius=9, shapeName=Dyn_cape_mid_01_r_capsule, center=<9 0 10.4477>, rotation=<rot 90 0 0>}]
 Dyn_cape_mid_01_r_MSTR [{shapeType=Capsule, length=0.505, radius=0.505, shapeName=Dyn_cape_mid_01_r_MSTR_capsule, rotation=<rot 0 -90 0>}]
 Dyn_cape_mid_02_l [{shapeType=Capsule, length=12, radius=12, shapeName=Dyn_cape_mid_02_l_capsule, center=<12 0 7.7477>, rotation=<rot 90 0 0>}]
 Dyn_cape_mid_02_r [{shapeType=Capsule, length=12, radius=12, shapeName=Dyn_cape_mid_02_r_capsule, center=<12 0 7.7477>, rotation=<rot 90 0 0>}]
 Dyn_cape_mid_03_l [
  {shapeType=Capsule, length=21.5, radius=8.5, shapeName=Dyn_cape_mid_03_l_capsule, center=<9 0 7>, rotation=<rot 90 -90 0>}
  {shapeType=Capsule, length=21.5, radius=8.5, shapeName=Dyn_cape_mid_03_l_capsule, center=<21.4265 0 7>, rotation=<rot 90 -90 0>}
 ]
 Dyn_cape_mid_03_r [
  {shapeType=Capsule, length=21.5, radius=8.5, shapeName=Dyn_cape_mid_03_r_capsule, center=<9 0 7>, rotation=<rot 90 -90 0>}
  {shapeType=Capsule, length=21.5, radius=8.5, shapeName=Dyn_cape_mid_03_r_capsule, center=<21.2158 0 7>, rotation=<rot 90 -90 0>}
 ]
 Dyn_cape_mid_04_l [{shapeType=Capsule, length=21.5, radius=9.2, shapeName=Dyn_cape_mid_04_l_capsule, center=<9 0 6.5>, rotation=<rot -0.0003 0 -85>}]
 Dyn_cape_mid_04_r [{shapeType=Capsule, length=21.5, radius=9.2, shapeName=Dyn_cape_mid_04_r_capsule, center=<9 0 6.5>, rotation=<rot -0.0003 0 -85>}]
 Dyn_cape_mid_05_l [{shapeType=Capsule, length=21.5, radius=9.2, shapeName=Dyn_cape_mid_05_l_capsule, center=<9 0 5.2>, rotation=<rot -0.0003 0 -85>}]
 Dyn_cape_mid_05_r [{shapeType=Capsule, length=21.5, radius=9.2, shapeName=Dyn_cape_mid_05_r_capsule, center=<9 0 5.2>, rotation=<rot -0.0003 0 -85>}]
 calf_l [{shapeType=Capsule, length=25.2372, radius=10.7837, center=<-17.038 -1.9066 -0.3219>, rotation=<rot 90 0 0>}]
 calf_r [{shapeType=Capsule, length=24.2372, radius=11.7837, center=<17.038 2.1081 0.3219>, rotation=<rot 90 180 0>}]
 clavicle_l [{shapeType=Capsule, length=7.0634, radius=6.5229, center=<7 -4 0>, rotation=<rot 90 0 0>}]
 clavicle_r [{shapeType=Capsule, length=7.0634, radius=6.5229, center=<-7 4 0>, rotation=<rot 90 180 0>}]
 foot_l [{shapeType=Capsule, length=13.9693, radius=8.9346, center=<-3.004 6.5342 -0.325>, rotation=<rot 0 0 90>}]
 foot_r [{shapeType=Capsule, length=13.9693, radius=8.9346, center=<3.004 -6.5342 0.325>, rotation=<rot 0 0 -90>}]
 hand_l [{shapeType=Capsule, length=9.1708, radius=6.2527, center=<7.7673 -2.6609 0.26>, rotation=<rot 0 81.1706 -79.3237>}]
 hand_r [{shapeType=Capsule, length=8.9257, radius=6.6213, center=<-7.7545 2.6833 -0.6535>, rotation=<rot 0 82.0056 -74.5376>}]
 lowerarm_l [{shapeType=Capsule, length=14.3712, radius=6.1956, center=<9.8422 0 0>, rotation=<rot 90 0 0>}]
 lowerarm_r [{shapeType=Capsule, length=14.3712, radius=6.1956, center=<-9.8422 0 0>, rotation=<rot 90 180 0>}]
 pelvis [{shapeType=Capsule, length=18.5, radius=12.5082, center=<0 -3.83 0>}]
 root [{shapeType=Box, extentX=258.5, extentY=266.5, extentZ=59.5, shapeName=root_box, center=<0 0 -29.8121>, rotation=<rot 0 -90 0>}]
 spine_01 [{shapeType=Capsule, length=8.005, radius=13.505, shapeName=spine_01_capsule, center=<0 -1.006 0>}]
 spine_02 [{shapeType=Capsule, length=19.3921, radius=10.8899, center=<0 -2.1939 0>}]
 spine_03 [{shapeType=Capsule, length=10.3957, radius=11.4522, center=<0 -1.3308 0>}]
 spine_04 [{shapeType=Capsule, length=12.7863, radius=9.6562, center=<0 -2.4745 0>}]
 spine_05 [{shapeType=Capsule, length=16.6149, radius=12.1005, center=<-1.7982 -0.4791 0>}]
 thigh_l [{shapeType=Capsule, length=19.3867, radius=11.2219, shapeName=thigh_l_capsule, center=<-18.1103 -3.7326 0>, rotation=<rot 90 0 0>}]
 thigh_r [{shapeType=Capsule, length=19.3867, radius=11.2219, shapeName=thigh_r_capsule, center=<18.1103 3.7326 0>, rotation=<rot 90 180 0>}]
 upperarm_l [{shapeType=Capsule, length=15.4485, radius=5.6809, center=<13.0051 -1 0>, rotation=<rot 90 0 0>}]
 upperarm_r [{shapeType=Capsule, length=15.4485, radius=5.6809, center=<-13.0051 1.8622 0>, rotation=<rot 90 180 0>}]
}
constraints [
 {
  bone1Name=spine_02
  bone2Name=spine_01
  swing1LimitDegrees=45
  swing1Motion=Locked
  swing2LimitDegrees=45
  swing2Motion=Locked
  twistLimitDegrees=45
  twistMotion=Locked
 }
 {
  bone1Name=spine_03
  bone2Name=spine_02
  swing1LimitDegrees=45
  swing1Motion=Locked
  swing2LimitDegrees=45
  swing2Motion=Locked
  twistLimitDegrees=45
  twistMotion=Locked
 }
 {
  bone1Name=spine_04
  bone2Name=spine_03
  swing1LimitDegrees=45
  swing1Motion=Locked
  swing2LimitDegrees=45
  swing2Motion=Locked
  twistLimitDegrees=45
  twistMotion=Locked
 }
 {
  bone1Name=spine_05
  bone2Name=spine_04
  swing1LimitDegrees=45
  swing1Motion=Locked
  swing2LimitDegrees=45
  swing2Motion=Locked
  twistLimitDegrees=45
  twistMotion=Locked
 }
 {
  bone1Name=clavicle_l
  bone2Name=spine_05
  swing1LimitDegrees=45
  swing1Motion=Locked
  swing2LimitDegrees=45
  swing2Motion=Locked
  twistLimitDegrees=45
  twistMotion=Locked
 }
 {
  bone1Name=upperarm_l
  bone2Name=clavicle_l
  swing1LimitDegrees=45
  swing1Motion=Locked
  swing2LimitDegrees=45
  swing2Motion=Locked
  twistLimitDegrees=45
  twistMotion=Locked
 }
 {
  bone1Name=lowerarm_l
  bone2Name=upperarm_l
  swing1LimitDegrees=45
  swing1Motion=Locked
  swing2LimitDegrees=45
  swing2Motion=Locked
  twistLimitDegrees=45
  twistMotion=Locked
 }
 {
  bone1Name=hand_l
  bone2Name=lowerarm_l
  swing1LimitDegrees=45
  swing1Motion=Locked
  swing2LimitDegrees=45
  swing2Motion=Locked
  twistLimitDegrees=45
  twistMotion=Locked
 }
 {
  bone1Name=clavicle_r
  bone2Name=spine_05
  swing1LimitDegrees=45
  swing1Motion=Locked
  swing2LimitDegrees=45
  swing2Motion=Locked
  twistLimitDegrees=45
  twistMotion=Locked
 }
 {
  bone1Name=upperarm_r
  bone2Name=clavicle_r
  swing1LimitDegrees=45
  swing1Motion=Locked
  swing2LimitDegrees=45
  swing2Motion=Locked
  twistLimitDegrees=45
  twistMotion=Locked
 }
 {
  bone1Name=lowerarm_r
  bone2Name=upperarm_r
  swing1LimitDegrees=45
  swing1Motion=Locked
  swing2LimitDegrees=45
  swing2Motion=Locked
  twistLimitDegrees=45
  twistMotion=Locked
 }
 {
  bone1Name=hand_r
  bone2Name=lowerarm_r
  swing1LimitDegrees=45
  swing1Motion=Locked
  swing2LimitDegrees=45
  swing2Motion=Locked
  twistLimitDegrees=45
  twistMotion=Locked
 }
 {
  bone1Name=foot_l
  bone2Name=calf_l
  swing1LimitDegrees=45
  swing1Motion=Locked
  swing2LimitDegrees=45
  swing2Motion=Locked
  twistLimitDegrees=45
  twistMotion=Locked
 }
 {
  bone1Name=calf_r
  bone2Name=thigh_r
  swing1LimitDegrees=45
  swing1Motion=Locked
  swing2LimitDegrees=45
  swing2Motion=Locked
  twistLimitDegrees=45
  twistMotion=Locked
 }
 {
  bone1Name=foot_r
  bone2Name=calf_r
  swing1LimitDegrees=45
  swing1Motion=Locked
  swing2LimitDegrees=45
  swing2Motion=Locked
  twistLimitDegrees=45
  twistMotion=Locked
 }
 {
  bone1Name=spine_01
  bone2Name=pelvis
  swing1LimitDegrees=45
  swing1Motion=Locked
  swing2LimitDegrees=45
  swing2Motion=Locked
  twistLimitDegrees=45
  twistMotion=Locked
 }
 {
  bone1Name=thigh_l
  bone2Name=pelvis
  swing1LimitDegrees=45
  swing1Motion=Locked
  swing2LimitDegrees=45
  swing2Motion=Locked
  twistLimitDegrees=45
  twistMotion=Locked
 }
 {
  bone1Name=thigh_r
  bone2Name=pelvis
  swing1LimitDegrees=45
  swing1Motion=Locked
  swing2LimitDegrees=45
  swing2Motion=Locked
  twistLimitDegrees=45
  twistMotion=Locked
 }
 {
  bone1Name=Dyn_cape_mid_01_l_MSTR
  bone2Name=spine_05
  swing1LimitDegrees=45
  swing1Motion=Locked
  swing2LimitDegrees=45
  swing2Motion=Locked
  twistLimitDegrees=45
  twistMotion=Locked
 }
 {bone1Name=Dyn_cape_mid_01_l, bone2Name=Dyn_cape_mid_01_l_MSTR, swing1LimitDegrees=10, swing2LimitDegrees=10, twistLimitDegrees=0}
 {bone1Name=Dyn_cape_mid_02_l, bone2Name=Dyn_cape_mid_01_l, swing1LimitDegrees=20, swing2LimitDegrees=20, twistLimitDegrees=0}
 {bone1Name=Dyn_cape_mid_03_l, bone2Name=Dyn_cape_mid_02_l, swing1LimitDegrees=20, swing2LimitDegrees=20, twistLimitDegrees=0}
 {bone1Name=Dyn_cape_mid_04_l, bone2Name=Dyn_cape_mid_03_l, swing1LimitDegrees=45, swing2LimitDegrees=30, twistLimitDegrees=0}
 {bone1Name=Dyn_cape_mid_05_l, bone2Name=Dyn_cape_mid_04_l, swing1LimitDegrees=45, swing2LimitDegrees=45, twistLimitDegrees=0}
 {
  bone1Name=Dyn_cape_mid_01_r_MSTR
  bone2Name=spine_05
  swing1LimitDegrees=45
  swing1Motion=Locked
  swing2LimitDegrees=45
  swing2Motion=Locked
  twistLimitDegrees=45
  twistMotion=Locked
 }
 {bone1Name=Dyn_cape_mid_01_r, bone2Name=Dyn_cape_mid_01_r_MSTR, swing1LimitDegrees=10, swing2LimitDegrees=10, twistLimitDegrees=0}
 {bone1Name=Dyn_cape_mid_02_r, bone2Name=Dyn_cape_mid_01_r, swing1LimitDegrees=20, swing2LimitDegrees=20, twistLimitDegrees=0}
 {bone1Name=Dyn_cape_mid_03_r, bone2Name=Dyn_cape_mid_02_r, swing1LimitDegrees=20, swing2LimitDegrees=20, twistLimitDegrees=0}
 {bone1Name=Dyn_cape_mid_04_r, bone2Name=Dyn_cape_mid_03_r, swing1LimitDegrees=45, swing2LimitDegrees=30, twistLimitDegrees=0}
 {bone1Name=Dyn_cape_mid_05_r, bone2Name=Dyn_cape_mid_04_r, swing1LimitDegrees=45, swing2LimitDegrees=45, twistLimitDegrees=0}
 {
  bone1Name=Dyn_cape_01_l1
  bone2Name=clavicle_l
  swing1LimitDegrees=45
  swing1Motion=Locked
  swing2LimitDegrees=45
  swing2Motion=Locked
  twistLimitDegrees=45
  twistMotion=Locked
 }
 {bone1Name=Dyn_cape_02_l, bone2Name=Dyn_cape_01_l1, swing1LimitDegrees=10, swing2LimitDegrees=10, twistLimitDegrees=0}
 {bone1Name=Dyn_cape_03_l, bone2Name=Dyn_cape_02_l, swing1LimitDegrees=20, swing2LimitDegrees=20, twistLimitDegrees=0}
 {bone1Name=Dyn_cape_04_l, bone2Name=Dyn_cape_03_l, swing1LimitDegrees=20, swing2LimitDegrees=20, twistLimitDegrees=0}
 {bone1Name=Dyn_cape_05_l, bone2Name=Dyn_cape_04_l, swing1LimitDegrees=45, swing2LimitDegrees=30, twistLimitDegrees=0}
 {bone1Name=Dyn_cape_06_l, bone2Name=Dyn_cape_05_l, swing1LimitDegrees=45, swing2LimitDegrees=45, twistLimitDegrees=0}
 {
  bone1Name=Dyn_cape_01_r1
  bone2Name=clavicle_r
  swing1LimitDegrees=45
  swing1Motion=Locked
  swing2LimitDegrees=45
  swing2Motion=Locked
  twistLimitDegrees=45
  twistMotion=Locked
 }
 {bone1Name=Dyn_cape_02_r, bone2Name=Dyn_cape_01_r1, swing1LimitDegrees=10, swing2LimitDegrees=10, twistLimitDegrees=0}
 {bone1Name=Dyn_cape_03_r, bone2Name=Dyn_cape_02_r, swing1LimitDegrees=20, swing2LimitDegrees=20, twistLimitDegrees=0}
 {bone1Name=Dyn_cape_04_r, bone2Name=Dyn_cape_03_r, swing1LimitDegrees=20, swing2LimitDegrees=20, twistLimitDegrees=0}
 {bone1Name=Dyn_cape_05_r, bone2Name=Dyn_cape_04_r, swing1LimitDegrees=45, swing2LimitDegrees=30, twistLimitDegrees=0}
 {bone1Name=Dyn_cape_06_r, bone2Name=Dyn_cape_05_r, swing1LimitDegrees=45, swing2LimitDegrees=45, twistLimitDegrees=0}
 {
  bone1Name=Dyn_cape_mid_01_l
  bone2Name=Dyn_cape_02_l
  swing1LimitDegrees=45
  swing1Motion=Free
  swing2LimitDegrees=45
  swing2Motion=Free
  twistLimitDegrees=45
  twistMotion=Free
 }
 {
  bone1Name=Dyn_cape_mid_01_r
  bone2Name=Dyn_cape_02_r
  swing1LimitDegrees=45
  swing1Motion=Free
  swing2LimitDegrees=45
  swing2Motion=Free
  twistLimitDegrees=45
  twistMotion=Free
 }
 {
  bone1Name=Dyn_cape_mid_01_r
  bone2Name=Dyn_cape_mid_01_l
  swing1LimitDegrees=45
  swing1Motion=Free
  swing2LimitDegrees=45
  swing2Motion=Free
  twistLimitDegrees=45
  twistMotion=Free
 }
 {
  bone1Name=Dyn_cape_mid_04_l
  bone2Name=Dyn_cape_05_l
  swing1LimitDegrees=45
  swing1Motion=Free
  swing2LimitDegrees=45
  swing2Motion=Free
  twistLimitDegrees=45
  twistMotion=Free
 }
 {
  bone1Name=Dyn_cape_mid_04_r
  bone2Name=Dyn_cape_05_r
  swing1LimitDegrees=45
  swing1Motion=Free
  swing2LimitDegrees=45
  swing2Motion=Free
  twistLimitDegrees=45
  twistMotion=Free
 }
 {
  bone1Name=Dyn_cape_mid_04_l
  bone2Name=Dyn_cape_mid_04_r
  swing1LimitDegrees=45
  swing1Motion=Free
  swing2LimitDegrees=45
  swing2Motion=Free
  twistLimitDegrees=45
  twistMotion=Free
 }
 {
  bone1Name=calf_l
  bone2Name=thigh_l
  swing1LimitDegrees=45
  swing1Motion=Locked
  swing2LimitDegrees=45
  swing2Motion=Locked
  twistLimitDegrees=45
  twistMotion=Locked
 }
]
