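# Control Rig for the SolveSolid male medium cosmetic body, adding secondary deformation to its fur, cape, shoulder pad and belts.
# It runs FortDeformRig, then Constraints (weighted parent constraints), Fur and Straps.
# Fur clamps bones outside spheres around the head and blends arm and head pose-reader correctives, plus cape and belt poses gated by GoldenGloryEmote.
# Straps drives belt correctives from thigh pose readers and pelvis height while IsCrouching.
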ControlRigBlueprint /BRCosmetics/Characters/Player/Male/Medium/Bodies/M_MED_SolveSolid/Meshes/M_MED_SolveSolid_CtrlRig
graphs [
 New Function_ContainedGraph {nodes [
   RigVMFunctionEntryNode()
   > &RigUnit_ParentConstraint(
    AdvancedSettings=(RotationOrderForFilter=XZY)
    Child=(Type=Bone,Name="base_belt_side_00_lf")
    Parents="((Item=(Type=Bone,Name=\"deform_glute_l\"),Weight=0.0113),(Item=(Type=Bone,Name=\"thigh_l\"),Weight=0.0369),(Item=(Type=Bone,Name=\"deform_groin_l\"),Weight=0.0699),(Item=(Type=Bone,Name=\"pelvis\"),Weight=0.1228),(Item=(Type=Bone,Name=\"thigh_twist_01_l\"),Weight=0 ...(truncated)"
    Weight=1
   )
   > &RigUnit_ParentConstraint(
    AdvancedSettings=(RotationOrderForFilter=XZY)
    Child=(Type=Bone,Name="base_belt_front_00_rt")
    Parents="((Item=(Type=Bone,Name=\"deform_groin_r\"),Weight=0),(Item=(Type=Bone,Name=\"thigh_twist_01_r\"),Weight=0),(Item=(Type=Bone,Name=\"spine_03\"),Weight=0),(Item=(Type=Bone,Name=\"spine_01\"),Weight=0),(Item=(Type=Bone,Name=\"spine_02\"),Weight=0),(Item=(Type=Bone,Name ...(truncated)"
    Weight=1
   )
   > &RigUnit_ParentConstraint(
    AdvancedSettings=(RotationOrderForFilter=XZY)
    Child=(Type=Bone,Name="base_belt_side_00_rt")
    Parents="((Item=(Type=Bone,Name=\"deform_glute_r\"),Weight=0.0058),(Item=(Type=Bone,Name=\"thigh_r\"),Weight=0.0064),(Item=(Type=Bone,Name=\"deform_groin_r\"),Weight=0.1775),(Item=(Type=Bone,Name=\"pelvis\"),Weight=0.2449),(Item=(Type=Bone,Name=\"thigh_twist_01_r\"),Weight=0 ...(truncated)"
    Weight=1
   )
   > &RigUnit_ParentConstraint(
    AdvancedSettings=(RotationOrderForFilter=XZY)
    Child=(Type=Bone,Name="base_shoulderpad_side_00_rt")
    Parents=((Item=(Type=Bone,Name="clavicle_r"),Weight=0),(Item=(Type=Bone,Name="upperarm_r"),Weight=1))
    Weight=1
   )
   > &RigUnit_ParentConstraint(
    AdvancedSettings=(RotationOrderForFilter=XZY)
    Child=(Type=Bone,Name="Dyn_cape_01_l1")
    Parents="((Item=(Type=Bone,Name=\"upperarm_twist_02_l\"),Weight=0.003),(Item=(Type=Bone,Name=\"spine_04\"),Weight=0.0099),(Item=(Type=Bone,Name=\"deform_upperarm_fr_l\"),Weight=0.0106),(Item=(Type=Bone,Name=\"spine_03\"),Weight=0.0416),(Item=(Type=Bone,Name=\"upperarm_twist ...(truncated)"
    Weight=1
   )
   > RigVMFunctionReturnNode()
   &RigUnit_ModifyTransforms(
    ExecutePin=()
    ItemToModify="((Item=(Type=Bone,Name=\"base_fur_front_00_rt\"),Transform=(Rotation=(X=0,Y=0.1305,Z=0,W=-0.9914),Translation=(X=-4.784,Y=0,Z=9.9576),Scale3D=(X=1,Y=1,Z=1))),(Item=(Type=Bone,Name=\"base_fur_back_00_rt\"),Transform=(Rotation=(X=0,Y=0,Z=0,W=-1),Translation=(X=0 ...(truncated)"
    Weight=1
    WeightMinimum=0
   )
   &RigUnit_ParentConstraint(
    AdvancedSettings=(RotationOrderForFilter=XZY)
    Child=(Type=Bone,Name="base_belt_front_00_rt")
    ExecutePin=()
    Parents="((Item=(Type=Bone,Name=\"deform_groin_r\"),Weight=0.0005),(Item=(Type=Bone,Name=\"thigh_twist_01_r\"),Weight=0.0058),(Item=(Type=Bone,Name=\"spine_03\"),Weight=0.0099),(Item=(Type=Bone,Name=\"spine_01\"),Weight=0.0153),(Item=(Type=Bone,Name=\"spine_02\"),Weight=0.01 ...(truncated)"
    Weight=1
   )
  ]}
 New Function_ContainedGraph {nodes [
   n1 = RigVMFunctionEntryNode()
   n2 = &RigUnit_GetTransform(bInitial=False, Item=@n1.Source)
   ClampSpatially = /Script/RigVM.RigVMFunction_MathTransformClampSpatially(
    Axis=None
    DebugColor=(R=1,G=0,B=0,A=1)
    DebugThickness=1
    Type=Sphere
    Space=&RigVMFunction_MathTransformMul(
     B=&RigUnit_GetTransform(bInitial=False, Item=@n1.Target).Transform
     A=/Script/RigVM.RigVMFunction_MathTransformMake(Rotation=(X=0,Y=0,Z=0,W=1), Scale=(X=1,Y=1,Z=1), Translation=@n1.Target Offset) {is=struct RigVMUnitNode}.Result
    ) {template="Multiply::Execute(in A,in B,out Result)"}.Result
    Value=&RigVMFunction_MathTransformMul(
     B=@n2.Transform
     A=&RigVMFunction_MathTransformMake(Rotation=(X=0,Y=0,Z=0,W=1), Scale=(X=1,Y=1,Z=1), Translation=@n1.Source Offset).Result
    ) {template="Multiply::Execute(in A,in B,out Result)"}.Result
    Maximum=@n1.Max Radius
    Minimum=@n1.Min Radius
    bDrawDebug=@n1.Debug
   ) {
    is=struct RigVMUnitNode
    template="ClampSpatially::Execute(in Value,in Axis,in Type,in Minimum,in Maximum,in Space,in bDrawDebug,in DebugColor,in DebugThickness,out Result)"
   }
   LessEqual = /Script/RigVM.RigVMFunction_MathDoubleLessEqual(B=@n1.Max Radius, A=&RigVMFunction_MathVectorDistance(B=@ClampSpatially.Result.Translation, A=@n2.Transform.Translation).Result) {is=struct RigVMUnitNode, template="LessEqual::Execute(in A,in B,out Result)"}
   n3 = &RigVMFunction_ControlFlowBranch(ExecuteContext=@n1.ExecuteContext, Condition=@LessEqual.Result)
   RigVMFunctionReturnNode(IsColliding=@LessEqual.Result, ExecuteContext=@n3.Completed)
   &RigUnit_ModifyTransforms(
    ItemToModify=((Item=(Type=Bone,Name="None"),Transform=(Rotation=(X=0,Y=0,Z=0,W=1),Translation=(X=0,Y=0,Z=0),Scale3D=(X=1,Y=1,Z=1))))
    Mode=OverrideGlobal
    WeightMinimum=0
    ItemToModify.0.Item=@n1.Source
    ExecutePin=@n3.True
    Weight=@n1.Weight
    ItemToModify.0.Transform=@ClampSpatially.Result
   )
  ]}
 New Function_ContainedGraph {nodes [
   n1 = RigVMFunctionEntryNode()
   RigVMFunction_MathBoolToInteger = &RigVMFunction_MathBoolToInteger(Value=@n1.onlyTwist)
   RigVMFunction_MathQuaternionSwingTwist = &RigVMFunction_MathQuaternionSwingTwist(
    twistAxis=@n1.twistAxis
    Input=&RigUnit_GetRelativeTransformForItem(
     bChildInitial=False
     bParentInitial=False
     child=(Type=Bone,Name="None")
     parent=(Type=Bone,Name="None")
     child.Name=@n1.child
     parent.Name=@n1.parent
     parent.Type=@n1.parentType
    ).RelativeTransform.Rotation
   )
   RigVMFunction_MathQuaternionToEuler = &RigVMFunction_MathQuaternionToEuler(
    Value=RigVMDispatchNode(
     Values=((),())
     Index=@RigVMFunction_MathBoolToInteger.Result
     Values.0=@RigVMFunction_MathQuaternionSwingTwist.Swing
     Values.1=@RigVMFunction_MathQuaternionSwingTwist.Twist
    ) {template="DISPATCH_RigVMDispatch_SelectInt32(in Index,in Values,out Result)"}.Result
    RotationOrder=@n1.Rotation Order
   )
   > DISPATCH_RigVMDispatch_Print = RigVMDispatchNode(ScreenColor=(R=1,G=1,B=1,A=1), ScreenDuration=1.5, Enabled=@n1.debugRotation, Value=@RigVMFunction_MathQuaternionToEuler.Result) {template="DISPATCH_RigVMDispatch_Print(in Prefix,in Value,in Enabled,in ScreenDuration,in ScreenColor)"}
   Clamp = &RigVMFunction_MathIntClamp(Value=@n1.primAxis)
   DISPATCH_RigVMDispatch_SelectInt32 = RigVMDispatchNode(
    Values=(0,0,0)
    Index=@Clamp.Result
    Values.0=@RigVMFunction_MathQuaternionToEuler.Result.X
    Values.1=@RigVMFunction_MathQuaternionToEuler.Result.Y
    Values.2=@RigVMFunction_MathQuaternionToEuler.Result.Z
   ) {template="DISPATCH_RigVMDispatch_SelectInt32(in Index,in Values,out Result)"}
   Remap = &RigVMFunction_MathDoubleRemap(
    TargetMaximum=1
    TargetMinimum=0
    bClamp=False
    Value=@DISPATCH_RigVMDispatch_SelectInt32.Result
    SourceMaximum=@n1.max
    SourceMinimum=@n1.min
   )
   Multiply = &RigVMFunction_MathDoubleMul(B=1, A=&RigVMFunction_MathDoubleClamp(Value=@Remap.Result).Result)
   > RigVMFunctionReturnNode(outputWeight=@Multiply.Result)
  ]}
 New Function_ContainedGraph {nodes [
   RigVMFunctionEntryNode()
   > RigVMFunction_Sequence = RigVMAggregateNode()
   Spherical Collision_1 = RigVMFunctionReferenceNode(
    Debug=true
    Max Radius=258
    Min Radius=10
    Source=(Type=Bone,Name="base_fur_back_00_ct")
    Source Offset=(X=0,Y=0,Z=0)
    Target=(Type=Bone,Name="head")
    Target Offset=(X=5,Y=-2,Z=0)
    Weight=1
    ExecuteContext=@RigVMFunction_Sequence.A
   )
   > Spherical Collision_2 = RigVMFunctionReferenceNode(
    Debug=false
    Max Radius=25
    Min Radius=16
    Source=(Type=Bone,Name="base_fur_back_00_lf")
    Source Offset=(X=0,Y=0,Z=0)
    Target=(Type=Bone,Name="head")
    Target Offset=(X=3,Y=2,Z=0)
    Weight=1
   )
   > Spherical Collision_3 = RigVMFunctionReferenceNode(
    Debug=false
    Max Radius=25
    Min Radius=16
    Source=(Type=Bone,Name="base_fur_back_00_rt")
    Source Offset=(X=0,Y=0,Z=0)
    Target=(Type=Bone,Name="head")
    Target Offset=(X=3,Y=2,Z=0)
    Weight=1
   )
   > Spherical Collision_4 = RigVMFunctionReferenceNode(
    Debug=false
    Max Radius=25
    Min Radius=16
    Source=(Type=Bone,Name="base_fur_front_00_lf")
    Source Offset=(X=0,Y=0,Z=0)
    Target=(Type=Bone,Name="head")
    Target Offset=(X=3,Y=2,Z=0)
    Weight=1
   )
   > Spherical Collision_5 = RigVMFunctionReferenceNode(
    Debug=false
    Max Radius=25
    Min Radius=16
    Source=(Type=Bone,Name="base_fur_front_00_rt")
    Source Offset=(X=0,Y=0,Z=0)
    Target=(Type=Bone,Name="head")
    Target Offset=(X=3,Y=2,Z=0)
    Weight=1
   )
   n1 = &RigUnit_SphericalPoseReader(
    ActiveRegionScaleFactors=(PositiveWidth=1,NegativeWidth=0.5,PositiveHeight=0.75,NegativeHeight=0.75)
    ActiveRegionSize=0.1
    Debug=(bDrawDebug=false,bDraw2D=False,bDrawLocalAxes=False,DebugScale=25,DebugSegments=20,DebugThickness=0.25)
    DriverAxis=(X=0,Y=1,Z=0)
    DriverItem=(Type=Bone,Name="head")
    FalloffRegionScaleFactors=(PositiveWidth=1,NegativeWidth=0.5,PositiveHeight=1,NegativeHeight=1)
    FalloffSize=0.2
    FlipHeightScaling=False
    FlipWidthScaling=False
    OptionalParentItem=(Type=Bone,Name="None")
    RotationOffset=(X=90,Y=0,Z=-30)
    ExecutePin=@RigVMFunction_Sequence.C
   )
   > n2 = &RigUnit_SphericalPoseReader(
    ActiveRegionScaleFactors=(PositiveWidth=1,NegativeWidth=1,PositiveHeight=1,NegativeHeight=1)
    ActiveRegionSize=0.025
    Debug=(bDrawDebug=false,bDraw2D=False,bDrawLocalAxes=False,DebugScale=25,DebugSegments=20,DebugThickness=0.25)
    DriverAxis=(X=-1,Y=0,Z=0)
    DriverItem=(Type=Bone,Name="upperarm_r")
    FalloffRegionScaleFactors=(PositiveWidth=1,NegativeWidth=1,PositiveHeight=1,NegativeHeight=1)
    FalloffSize=0.1
    FlipHeightScaling=False
    FlipWidthScaling=False
    OptionalParentItem=(Type=Bone,Name="None")
    RotationOffset=(X=-115,Y=0,Z=25)
   )
   Multiply = &RigVMFunction_MathDoubleMul(B=@n1.OutputParam, A=@n2.OutputParam)
   > &RigUnit_ModifyTransforms(
    ItemToModify="((Item=(Type=Bone,Name=\"base_fur_front_00_lf\"),Transform=(Rotation=(X=-0.0441,Y=-0.0031,Z=-0.0874,W=0.9952),Translation=(X=-1.8,Y=-0.3,Z=1.5),Scale3D=(X=1,Y=1,Z=1))),(Item=(Type=Bone,Name=\"base_fur_back_00_lf\"),Transform=(Rotation=(X=0,Y=0,Z=0,W=1),Transla ...(truncated)"
    WeightMinimum=0
    Weight=@Multiply.Result
   )
   RigVMFunction_ControlFlowBranch = &RigVMFunction_ControlFlowBranch(ExecuteContext=@RigVMFunction_Sequence.D, Condition=RigVMVariableNode() {variable=GoldenGloryEmote}.Value)
   n3 = &RigUnit_SphericalPoseReader(
    ActiveRegionScaleFactors=(PositiveWidth=2.5,NegativeWidth=3,PositiveHeight=3,NegativeHeight=3)
    ActiveRegionSize=0.03
    Debug=(bDrawDebug=false,bDraw2D=False,bDrawLocalAxes=False,DebugScale=25,DebugSegments=20,DebugThickness=0.25)
    DriverAxis=(X=1,Y=0,Z=0)
    DriverItem=(Type=Bone,Name="upperarm_l")
    FalloffRegionScaleFactors=(PositiveWidth=1,NegativeWidth=1,PositiveHeight=1,NegativeHeight=1)
    FalloffSize=0.3
    FlipHeightScaling=False
    FlipWidthScaling=False
    OptionalParentItem=(Type=Bone,Name="None")
    RotationOffset=(X=0,Y=90,Z=150)
    ExecutePin=@RigVMFunction_ControlFlowBranch.True
   )
   > &RigUnit_ModifyTransforms(
    ItemToModify="((Item=(Type=Bone,Name=\"Dyn_cape_01_l1\"),Transform=(Rotation=(X=-0.167,Y=0.3814,Z=0.4185,W=-0.8072),Translation=(X=0,Y=0,Z=0),Scale3D=(X=1,Y=1,Z=1))),(Item=(Type=Bone,Name=\"cape_01_l\"),Transform=(Rotation=(X=-0.0617,Y=-0.311,Z=0.4028,W=-0.8586),Translation ...(truncated)"
    WeightMinimum=0
    Weight=@n3.OutputParam
   )
   > n4 = &RigUnit_SphericalPoseReader(
    ActiveRegionScaleFactors=(PositiveWidth=2.5,NegativeWidth=3,PositiveHeight=3,NegativeHeight=3)
    ActiveRegionSize=0.03
    Debug=(bDrawDebug=true,bDraw2D=False,bDrawLocalAxes=False,DebugScale=25,DebugSegments=20,DebugThickness=0.25)
    DriverAxis=(X=-1,Y=0,Z=0)
    DriverItem=(Type=Bone,Name="upperarm_r")
    FalloffRegionScaleFactors=(PositiveWidth=1,NegativeWidth=1,PositiveHeight=1,NegativeHeight=1)
    FalloffSize=0.3
    FlipHeightScaling=false
    FlipWidthScaling=true
    OptionalParentItem=(Type=Bone,Name="None")
    RotationOffset=(X=-180,Y=90,Z=150)
   )
   > &RigUnit_ModifyTransforms(
    ItemToModify=((Item=(Type=Bone,Name="Dyn_cape_mid_01_r_MSTR"),Transform=(Rotation=(X=0.186,Y=0.1938,Z=-0.667,W=-0.695),Translation=(X=0,Y=0,Z=0),Scale3D=(X=1,Y=1,Z=1))))
    WeightMinimum=0
    Weight=@n4.OutputParam
   )
   > &RigUnit_ModifyTransforms(
    ItemToModify=((Item=(Type=Bone,Name="base_belt_side_00_rt"),Transform=(Rotation=(X=0,Y=0.1736,Z=0,W=0.9848),Translation=(X=0,Y=0,Z=0),Scale3D=(X=1,Y=1,Z=1))))
    Weight=1
    WeightMinimum=0
   )
   RigVMFunctionReturnNode(ExecuteContext=@RigVMFunction_Sequence.E)
   n5 = &RigUnit_SphericalPoseReader(
    ActiveRegionScaleFactors=(PositiveWidth=2.5,NegativeWidth=3,PositiveHeight=3,NegativeHeight=3)
    ActiveRegionSize=0.1
    Debug=(bDrawDebug=false,bDraw2D=False,bDrawLocalAxes=False,DebugScale=25,DebugSegments=20,DebugThickness=0.25)
    DriverAxis=(X=1,Y=0,Z=0)
    DriverItem=(Type=Bone,Name="upperarm_l")
    FalloffRegionScaleFactors=(PositiveWidth=1,NegativeWidth=1,PositiveHeight=1,NegativeHeight=1)
    FalloffSize=0.5
    FlipHeightScaling=False
    FlipWidthScaling=False
    OptionalParentItem=(Type=Bone,Name="None")
    RotationOffset=(X=0,Y=0,Z=0)
    ExecutePin=@RigVMFunction_Sequence.B
   )
   > &RigUnit_ModifyTransforms(
    ItemToModify="((Item=(Type=Bone,Name=\"base_fur_front_00_lf\"),Transform=(Rotation=(X=0,Y=-0.0218,Z=0,W=0.9998),Translation=(X=-2.3,Y=0,Z=2),Scale3D=(X=1,Y=1,Z=1))),(Item=(Type=Bone,Name=\"dyn_fur_back_00_lf\"),Transform=(Rotation=(X=0,Y=0,Z=0.0218,W=0.9998),Translation=(X= ...(truncated)"
    WeightMinimum=0
    Weight=@n5.OutputParam
   )
   > n6 = &RigUnit_SphericalPoseReader(
    ActiveRegionScaleFactors=(PositiveWidth=2.5,NegativeWidth=3,PositiveHeight=3,NegativeHeight=3)
    ActiveRegionSize=0.1
    Debug=(bDrawDebug=false,bDraw2D=False,bDrawLocalAxes=False,DebugScale=25,DebugSegments=20,DebugThickness=0.25)
    DriverAxis=(X=-1,Y=0,Z=0)
    DriverItem=(Type=Bone,Name="upperarm_r")
    FalloffRegionScaleFactors=(PositiveWidth=1,NegativeWidth=1,PositiveHeight=1,NegativeHeight=1)
    FalloffSize=0.5
    FlipHeightScaling=False
    FlipWidthScaling=False
    OptionalParentItem=(Type=Bone,Name="None")
    RotationOffset=(X=180,Y=0,Z=180)
   )
   > &RigUnit_ModifyTransforms(
    ItemToModify="((Item=(Type=Bone,Name=\"base_fur_front_00_rt\"),Transform=(Rotation=(X=-0.0252,Y=-0.0442,Z=0.0206,W=0.9985),Translation=(X=-2,Y=0,Z=2.1),Scale3D=(X=0.8,Y=1,Z=1))),(Item=(Type=Bone,Name=\"base_fur_back_00_rt\"),Transform=(Rotation=(X=0,Y=0,Z=0.0218,W=0.9998),T ...(truncated)"
    WeightMinimum=0
    Weight=@n6.OutputParam
   )
  ]}
 New Function_ContainedGraph {nodes [
   RigVMFunctionEntryNode()
   > RigVMFunction_Sequence = RigVMAggregateNode()
   n1 = &RigUnit_SphericalPoseReader(
    ActiveRegionScaleFactors=(PositiveWidth=1,NegativeWidth=1,PositiveHeight=1,NegativeHeight=1)
    ActiveRegionSize=0.05
    Debug=(bDrawDebug=false,bDraw2D=False,bDrawLocalAxes=False,DebugScale=25,DebugSegments=20,DebugThickness=0.25)
    DriverAxis=(X=1,Y=0,Z=0)
    DriverItem=(Type=Bone,Name="thigh_r")
    FalloffRegionScaleFactors=(PositiveWidth=1,NegativeWidth=1,PositiveHeight=1,NegativeHeight=1)
    FalloffSize=0.63
    FlipHeightScaling=true
    FlipWidthScaling=true
    OptionalParentItem=(Type=Bone,Name="None")
    RotationOffset=(X=-90,Y=0,Z=-30)
    ExecutePin=@RigVMFunction_Sequence.A
   )
   > &RigUnit_ModifyTransforms(
    ItemToModify=((Item=(Type=Bone,Name="None"),Transform=(Translation=(X=0,Y=0,Z=0),Scale3D=(X=1,Y=1,Z=1))))
    WeightMinimum=0
    Weight=@n1.OutputParam
   )
   > n2 = &RigUnit_SphericalPoseReader(
    ActiveRegionScaleFactors=(PositiveWidth=1,NegativeWidth=1,PositiveHeight=1,NegativeHeight=1)
    ActiveRegionSize=0.05
    Debug=(bDrawDebug=false,bDraw2D=False,bDrawLocalAxes=False,DebugScale=25,DebugSegments=20,DebugThickness=0.25)
    DriverAxis=(X=1,Y=0,Z=0)
    DriverItem=(Type=Bone,Name="thigh_r")
    FalloffRegionScaleFactors=(PositiveWidth=1,NegativeWidth=1,PositiveHeight=1,NegativeHeight=1)
    FalloffSize=0.67
    FlipHeightScaling=False
    FlipWidthScaling=False
    OptionalParentItem=(Type=Bone,Name="None")
    RotationOffset=(X=0,Y=30,Z=0)
   )
   > &RigUnit_ModifyTransforms(
    ItemToModify=((Item=(Type=Bone,Name="base_belt_side_00_rt"),Transform=(Rotation=(X=0,Y=0,Z=-0.4226,W=0.9063),Translation=(X=0,Y=0,Z=0),Scale3D=(X=1,Y=1,Z=1))))
    WeightMinimum=0
    Weight=@n2.OutputParam
   )
   > n3 = &RigUnit_SphericalPoseReader(
    ActiveRegionScaleFactors=(PositiveWidth=1,NegativeWidth=1,PositiveHeight=1,NegativeHeight=1)
    ActiveRegionSize=0.05
    Debug=(bDrawDebug=false,bDraw2D=False,bDrawLocalAxes=False,DebugScale=25,DebugSegments=20,DebugThickness=0.25)
    DriverAxis=(X=1,Y=0,Z=0)
    DriverItem=(Type=Bone,Name="thigh_r")
    FalloffRegionScaleFactors=(PositiveWidth=1,NegativeWidth=1,PositiveHeight=0.9,NegativeHeight=0.9)
    FalloffSize=0.63
    FlipHeightScaling=False
    FlipWidthScaling=False
    OptionalParentItem=(Type=Bone,Name="None")
    RotationOffset=(X=-40,Y=30,Z=0)
   )
   > &RigUnit_OffsetTransformForItem(
    Item=(Type=Bone,Name="None")
    OffsetTransform=(Rotation=(X=0,Y=0,Z=-0.4329,W=0.9014),Translation=(X=0,Y=0,Z=0),Scale3D=(X=1,Y=1,Z=1))
    bPropagateToChildren=True
    Weight=@n3.OutputParam
   )
   > n4 = &RigUnit_SphericalPoseReader(
    ActiveRegionScaleFactors=(PositiveWidth=1,NegativeWidth=1,PositiveHeight=1,NegativeHeight=1)
    ActiveRegionSize=0.05
    Debug=(bDrawDebug=false,bDraw2D=False,bDrawLocalAxes=False,DebugScale=25,DebugSegments=20,DebugThickness=0.25)
    DriverAxis=(X=1,Y=0,Z=0)
    DriverItem=(Type=Bone,Name="thigh_r")
    FalloffRegionScaleFactors=(PositiveWidth=0.75,NegativeWidth=0.75,PositiveHeight=0.9,NegativeHeight=0.9)
    FalloffSize=0.63
    FlipHeightScaling=False
    FlipWidthScaling=False
    OptionalParentItem=(Type=Bone,Name="None")
    RotationOffset=(X=90,Y=0,Z=10)
   )
   > &RigUnit_ModifyTransforms(
    ItemToModify=((Item=(Type=Bone,Name="None"),Transform=(Translation=(X=0,Y=0,Z=0),Scale3D=(X=1,Y=1,Z=1))))
    WeightMinimum=0
    Weight=@n4.OutputParam
   )
   > n5 = &RigUnit_SphericalPoseReader(
    ActiveRegionScaleFactors=(PositiveWidth=1,NegativeWidth=1,PositiveHeight=1,NegativeHeight=1)
    ActiveRegionSize=0.05
    Debug=(bDrawDebug=false,bDraw2D=False,bDrawLocalAxes=False,DebugScale=25,DebugSegments=20,DebugThickness=0.25)
    DriverAxis=(X=1,Y=0,Z=0)
    DriverItem=(Type=Bone,Name="thigh_r")
    FalloffRegionScaleFactors=(PositiveWidth=1,NegativeWidth=1,PositiveHeight=0.9,NegativeHeight=0.9)
    FalloffSize=0.63
    FlipHeightScaling=False
    FlipWidthScaling=False
    OptionalParentItem=(Type=Bone,Name="None")
    RotationOffset=(X=45,Y=0,Z=0)
   )
   > &RigUnit_OffsetTransformForItem(
    Item=(Type=Bone,Name="None")
    OffsetTransform=(Rotation=(X=0.0807,Y=0.2723,Z=-0.2723,W=0.9193),Translation=(X=0,Y=0,Z=0),Scale3D=(X=1,Y=1,Z=1))
    bPropagateToChildren=True
    Weight=@n5.OutputParam
   )
   n6 = &RigUnit_SphericalPoseReader(
    ActiveRegionScaleFactors=(PositiveWidth=1,NegativeWidth=1,PositiveHeight=1,NegativeHeight=1)
    ActiveRegionSize=0.05
    Debug=(bDrawDebug=false,bDraw2D=False,bDrawLocalAxes=False,DebugScale=25,DebugSegments=20,DebugThickness=0.25)
    DriverAxis=(X=-1,Y=0,Z=0)
    DriverItem=(Type=Bone,Name="thigh_l")
    FalloffRegionScaleFactors=(PositiveWidth=1,NegativeWidth=1,PositiveHeight=0.9,NegativeHeight=0.9)
    FalloffSize=0.63
    FlipHeightScaling=False
    FlipWidthScaling=False
    OptionalParentItem=(Type=Bone,Name="None")
    RotationOffset=(X=90,Y=0,Z=-30)
    ExecutePin=@RigVMFunction_Sequence.D
   )
   > &RigUnit_ModifyTransforms(
    ItemToModify=((Item=(Type=Bone,Name="base_belt_front_00_rt"),Transform=(Rotation=(X=0,Y=0,Z=0.5373,W=-0.8434),Translation=(X=0,Y=0,Z=0),Scale3D=(X=1,Y=1,Z=1))))
    WeightMinimum=0
    Weight=@n6.OutputParam
   )
   RigVMFunctionReturnNode(ExecuteContext=@RigVMFunction_Sequence.E)
   Remap = &RigVMFunction_MathDoubleRemap(
    SourceMaximum=45
    SourceMinimum=89
    TargetMaximum=1
    TargetMinimum=0
    Value=&RigUnit_GetRelativeTransformForItem(Child=(Type=Bone,Name="pelvis"), Parent=(Type=Bone,Name="root"), bChildInitial=False, bParentInitial=False).RelativeTransform.Translation.Z
   )
   RigVMFunction_MathBoolToFloat = &RigVMFunction_MathBoolToFloat(Value=RigVMVariableNode() {variable=IsCrouching}.Value)
   Multiply = &RigVMFunction_MathDoubleMul(A=@Remap.Result, B=@RigVMFunction_MathBoolToFloat.Result)
   &RigUnit_ModifyTransforms(
    ItemToModify="((Item=(Type=Bone,Name=\"base_belt_side_00_rt\"),Transform=(Rotation=(X=-0.0167,Y=0.2162,Z=0.0403,W=-0.9754),Translation=(X=0,Y=0,Z=0),Scale3D=(X=1,Y=1,Z=1))),(Item=(Type=Bone,Name=\"dyn_belt_side_01_rt\"),Transform=(Rotation=(X=0.0871,Y=-0.1539,Z=0.2197,W=0.9 ...(truncated)"
    WeightMinimum=0
    Weight=@Multiply.Result
    ExecutePin=@RigVMFunction_Sequence.C
   )
  ]}
 RigVMFunctionLibrary {nodes [
   Constraints = RigVMCollapseNode(ExecuteContext=())
   DeltaRotator = RigVMCollapseNode(ExecuteContext=(), debugRotation=False, onlyTwist=False, twistAxis=(X=0,Y=0,Z=0))
   Fur = RigVMCollapseNode(ExecuteContext=())
   Spherical Collision = RigVMCollapseNode(
    Debug=true
    ExecuteContext=()
    Max Radius=10
    Min Radius=9
    Source=(Type=None,Name="None")
    Source Offset=(X=0,Y=0,Z=0)
    Target=(Type=None,Name="None")
    Target Offset=(X=0,Y=0,Z=0)
    Weight=1
   )
   Straps = RigVMCollapseNode(ExecuteContext=())
  ]}
 RigVMFunction_Sequence_1_ContainedGraph {nodes [
   n1 = RigVMFunctionEntryNode()
   RigVMFunction_Sequence = &RigVMFunction_Sequence(ExecuteContext=@n1.ExecuteContext)
   n2 = &RigVMFunction_Sequence(ExecuteContext=@RigVMFunction_Sequence.B)
   n3 = &RigVMFunction_Sequence(ExecuteContext=@n2.B)
   n4 = &RigVMFunction_Sequence(ExecuteContext=@n3.B)
   RigVMFunctionReturnNode(A=@RigVMFunction_Sequence.A, B=@n2.A, C=@n3.A, D=@n4.A, E=@n4.B)
  ]}
 RigVMFunction_Sequence_1_ContainedGraph {nodes [
   n1 = RigVMFunctionEntryNode()
   RigVMFunction_Sequence = &RigVMFunction_Sequence(ExecuteContext=@n1.ExecuteContext)
   n2 = &RigVMFunction_Sequence(ExecuteContext=@RigVMFunction_Sequence.B)
   n3 = &RigVMFunction_Sequence(ExecuteContext=@n2.B)
   RigVMFunctionReturnNode(A=@RigVMFunction_Sequence.A, C=@n2.A, D=@n3.A, E=@n3.B)
  ]}
 RigVMModel {nodes [
   RigUnit_BeginExecution = &RigUnit_BeginExecution()
   /Script/FortniteGame.RigUnit_FortDeformRig(
    Bool=(bUpperArm=True,bElbow=True,bWrist=True,bGroin=True,bGlute=True,bKnee=True,bPec=True,bBicep=True,bLat=True)
    Curve="(bLUpperArmUpAlpha=False,bRUpperArmUpAlpha=False,bLUpperArmFrAlpha=False,bRUpperArmFrAlpha=False,bLUpperArmBkAlpha=False,bRUpperArmBkAlpha=False,bLWristUpAlpha=False,bRWristUpAlpha=False,bLWristDnAlpha=False,bRWristDnAlpha=False,bLKneeAlpha=False,bRKneeAlp ...(truncated)"
    Debug=(bUpperArmDebug=False,bWristDebug=False,bLatDebug=False,bGroinGluteDebug=False)
    Value="(UpperArmUp=(Rotation=(X=0.0,Y=-0.0,Z=0.0,W=1.0),Translation=(X=0.0,Y=0.0,Z=8.0),Scale3D=(X=0.70,Y=1.0,Z=1.0)),UpperArmFwdFr=(Rotation=(X=0.0,Y=-0.0,Z=0.0,W=1.0),Translation=(X=0.0,Y=10.0,Z=0.0),Scale3D=(X=0.50,Y=1.0,Z=1.0)),UpperArmFwdBk=(Rotation=(X=0.0, ...(truncated)"
    ExecutePin=@RigUnit_BeginExecution.ExecutePin
   ) {is=struct RigVMUnitNode}
   > Constraints = RigVMFunctionReferenceNode()
   > Fur = RigVMFunctionReferenceNode()
   > Straps = RigVMFunctionReferenceNode()
  ]}
]
hierarchy [
 root {loc=<0 0 0>, rot=<0 0 0 1>}
 pelvis {parent=root, loc=<0 -0.7071 93.1191>, rot=<0.0127 0.707 -0.0127 -0.707>}
 spine_01 {parent=pelvis, loc=<6.7494 -0.6388 0>, rot=<0 0 0.034 -0.9994>}
 spine_02 {parent=spine_01, loc=<7.2518 -0.035 0>, rot=<0 0 0.0238 -0.9997>}
 spine_03 {parent=spine_02, loc=<6.0104 0 0>, rot=<0 0 0.0149 -0.9999>}
 spine_04 {parent=spine_03, loc=<8.1024 0 0>, rot=<0 0 -0.0132 -0.9999>}
 spine_05 {parent=spine_04, loc=<10.0146 0 0>, rot=<0 0 -0.0015 -1>}
 neck_01 {parent=spine_05, loc=<11.3138 0.1529 0>, rot=<0 0 -0.2003 -0.9797>}
 neck_02 {parent=neck_01, loc=<4.5627 0 0>, rot=<0 0 0.0504 -0.9987>}
 head {parent=neck_02, loc=<5.3914 0 0>, rot=<0 0 0.115 -0.9934>}
 base_fur_back_00_ct {parent=spine_05, loc=<24.0689 -10.0222 -0.3699>, rot=<0.0284 -0.7065 -0.0284 -0.7065>}
 dyn_fur_back_00_ct {parent=base_fur_back_00_ct, loc=<0 0 0>, rot=<0 0 0 1>}
 base_fur_back_00_lf {parent=spine_05, loc=<18.8831 -3.5985 -15.9497>, rot=<0.2479 -0.6622 0.194 -0.68>}
 dyn_fur_back_00_lf {parent=base_fur_back_00_lf, loc=<0 0 0>, rot=<0 0 0 1>}
 base_fur_back_00_rt {parent=spine_05, loc=<18.8831 -3.5985 16.0821>, rot=<0.2479 -0.6622 0.194 -0.68>}
 dyn_fur_back_00_rt {parent=base_fur_back_00_rt, loc=<0 0 0>, rot=<0 0 0 1>}
 attach_cape {parent=spine_05, loc=<9.017 -11.714 0>, rot=<0.0616 -0.7044 -0.0616 -0.7044>}
 cape_mid_01 {parent=attach_cape, loc=<0 -0.1927 0.3708>, rot=<0.0664 -0.704 0.0664 -0.704>}
 cape_mid_02 {parent=cape_mid_01, loc=<13.0205 -0.6679 0>, rot=<0 0 -0.0917 -0.9958>}
 cape_mid_03 {parent=cape_mid_02, loc=<17.2 0 0>, rot=<0 0 -0.0085 -1>}
 cape_mid_04 {parent=cape_mid_03, loc=<17.2 0 0>, rot=<0 0 -0.0293 -0.9996>}
 cape_mid_05 {parent=cape_mid_04, loc=<17.2 0 0>, rot=<0 0 0.0005 -1>}
 cape_mid_06 {parent=cape_mid_05, loc=<17.2 0 0>, rot=<0 0 0.0016 -1>}
 cape_mid_07 {parent=cape_mid_06, loc=<17.2 0 0>, rot=<0 0 -0.0176 -0.9998>}
 cape_01_r {parent=cape_mid_01, loc=<-1.6798 2.0458 -16.1835>, rot=<-0.1563 -0.0473 0.0084 -0.9865>}
 cape_02_r {parent=cape_01_r, loc=<13.4515 -0.5279 0.0125>, rot=<0.0249 -0.0008 -0.11 -0.9936>}
 cape_03_r {parent=cape_02_r, loc=<17 0.0001 0>, rot=<0.0001 -0.0061 -0.0072 -1>}
 cape_04_r {parent=cape_03_r, loc=<17 0 0.0001>, rot=<0 -0.0024 -0.0118 -0.9999>}
 cape_05_r {parent=cape_04_r, loc=<17 -0.0001 -0.0001>, rot=<0 -0.0015 -0.0074 -1>}
 cape_06_r {parent=cape_05_r, loc=<17 0 0.0001>, rot=<0 0.0003 0.0015 -1>}
 cape_07_r {parent=cape_06_r, loc=<17 0.0001 -0.0001>, rot=<0 -0.0005 -0.0026 -1>}
 cape_01_l {parent=cape_mid_01, loc=<0.5223 3.4352 20.9134>, rot=<0.1549 0.0442 -0.0067 -0.9869>}
 cape_02_l {parent=cape_01_l, loc=<17 0 0.0001>, rot=<-0.0249 -0.0001 -0.0738 -0.997>}
 cape_03_l {parent=cape_02_l, loc=<16.9993 0 0>, rot=<0 0.0061 -0.03 -0.9995>}
 cape_04_l {parent=cape_03_l, loc=<17.0002 0 -0.0001>, rot=<0 0.0024 -0.0118 -0.9999>}
 cape_05_l {parent=cape_04_l, loc=<17 -0.0001 0.0001>, rot=<0 0.0015 -0.0074 -1>}
 cape_06_l {parent=cape_05_l, loc=<17 0 -0.0001>, rot=<0 -0.0003 0.0015 -1>}
 cape_07_l {parent=cape_06_l, loc=<17 0.0001 0.0002>, rot=<0 0.0005 -0.0026 -1>}
 cape {parent=attach_cape, type=Null, loc=<0 0 0>, rot=<0 0 0 1>}
 Dyn_backpack_root1 {parent=attach_cape, loc=<0 -1.9859 0.1921>, rot=<-0.0956 0 0 -0.9954>}
 Dyn_cape_mid_01_r_MSTR {parent=Dyn_backpack_root1, loc=<-9.1594 -0.0983 -1.3691>, rot=<-0.4529 -0.4587 0.5466 -0.5346>}
 Dyn_cape_mid_01_r {parent=Dyn_cape_mid_01_r_MSTR, loc=<0.2533 -0.0174 -0.0525>, rot=<0 0.0623 0 -0.9981>}
 Dyn_cape_mid_02_r {parent=Dyn_cape_mid_01_r, loc=<15.6773 0 -0.1591>, rot=<0 -0.0763 0 -0.9971>}
 Dyn_cape_mid_03_r {parent=Dyn_cape_mid_02_r, loc=<28.1116 0 -0.2126>, rot=<0 -0.0234 0 -0.9997>}
 Dyn_cape_mid_04_r {parent=Dyn_cape_mid_03_r, loc=<25.1969 0 0>, rot=<0 -0.0051 0 -1>}
 Dyn_cape_mid_05_r {parent=Dyn_cape_mid_04_r, loc=<22.2028 0 0>, rot=<0 0 0 1>}
 Dyn_cape_mid_06_r {parent=Dyn_cape_mid_05_r, loc=<13.3811 0 0>, rot=<0 0 0 1>}
 Dyn_cape_mid_01_l_MSTR {parent=Dyn_backpack_root1, loc=<9.1594 -0.0982 -1.3694>, rot=<-0.4587 -0.4529 0.5346 -0.5466>}
 Dyn_cape_mid_01_l {parent=Dyn_cape_mid_01_l_MSTR, loc=<0.2529 0.0174 -0.0524>, rot=<0 0.0623 0 -0.9981>}
 Dyn_cape_mid_02_l {parent=Dyn_cape_mid_01_l, loc=<15.6774 0 -0.1591>, rot=<0 -0.0763 0 -0.9971>}
 Dyn_cape_mid_03_l {parent=Dyn_cape_mid_02_l, loc=<28.1116 0 -0.2127>, rot=<0 -0.0234 0 -0.9997>}
 Dyn_cape_mid_04_l {parent=Dyn_cape_mid_03_l, loc=<25.1969 0 0>, rot=<0 -0.0051 0 -1>}
 Dyn_cape_mid_05_l {parent=Dyn_cape_mid_04_l, loc=<22.2027 0 0>, rot=<0 0 0 1>}
 Dyn_cape_mid_06_l {parent=Dyn_cape_mid_05_l, loc=<13.3811 0 0>, rot=<0 0 0 1>}
 attach_backpack {parent=spine_05, loc=<-0.123 -13.473 0>, rot=<-0.0093 -0.707 0.0093 -0.707>}
 Backpack_BR {parent=attach_backpack, type=Null, loc=<0 0 0>, rot=<0 0 0 1>}
 clavicle_l {parent=spine_05, loc=<4.3575 -0.1466 -2.3223>, rot=<0.0465 -0.7289 -0.009 -0.6829>}
 upperarm_l {parent=clavicle_l, loc=<15.5111 0 0>, rot=<0.023 -0.364 -0.0078 -0.9311>}
 lowerarm_l {parent=upperarm_l, loc=<26.6514 0 0>, rot=<0.0391 -0.0099 -0.2453 -0.9686>}
 hand_l {parent=lowerarm_l, loc=<22.3491 0 0>, rot=<-0.7081 -0.0507 0.0975 -0.6975>}
 thumb_01_l {parent=hand_l, loc=<1.3118 -0.3191 -2.459>, rot=<0.6546 -0.4042 0.0664 -0.6354>}
 thumb_02_l {parent=thumb_01_l, loc=<5.9336 -0.0448 0.0401>, rot=<-0.0037 0.0213 0.298 -0.9543>}
 thumb_03_l {parent=thumb_02_l, loc=<2.2223 0.3682 -0.0907>, rot=<0.0118 -0.013 0.0066 -0.9998>}
 index_metacarpal_l {parent=hand_l, loc=<1.6535 0.0183 -2.96>, rot=<0.0158 -0.0124 -0.0173 -0.9996>}
 index_01_l {parent=index_metacarpal_l, loc=<8.3008 0.0357 -0.0524>, rot=<-0.0044 0.0485 0.0924 -0.9945>}
 index_02_l {parent=index_01_l, loc=<3.3587 -0.3004 0.0452>, rot=<-0.002 0.0027 0.3225 -0.9466>}
 index_03_l {parent=index_02_l, loc=<2.5578 -0.123 -0.0252>, rot=<-0.002 0.0067 0.2558 -0.9667>}
 middle_metacarpal_l {parent=hand_l, loc=<2.1856 -0.1725 -0.5984>, rot=<0.006 0.0168 -0.0329 -0.9993>}
 middle_01_l {parent=middle_metacarpal_l, loc=<7.3275 -0.2447 0.0129>, rot=<-0.0208 0.0042 0.0858 -0.9961>}
 middle_02_l {parent=middle_01_l, loc=<4.0123 0.1401 0.1175>, rot=<-0.0218 0.0258 0.3512 -0.9357>}
 middle_03_l {parent=middle_02_l, loc=<2.6077 0.0534 0.1373>, rot=<0.0082 0.0188 0.1708 -0.9851>}
 ring_metacarpal_l {parent=hand_l, loc=<2.2074 -0.3625 1.3638>, rot=<-0.0284 0.0722 -0.0458 -0.9959>}
 ring_01_l {parent=ring_metacarpal_l, loc=<7.0295 -0.6197 -0.1446>, rot=<-0.0489 -0.0252 0.1892 -0.9804>}
 ring_02_l {parent=ring_01_l, loc=<3.6174 0.4647 -0.1027>, rot=<0 0 0.2806 -0.9598>}
 ring_03_l {parent=ring_02_l, loc=<2.4152 0.114 0.0199>, rot=<0 0 0.1189 -0.9929>}
 pinky_metacarpal_l {parent=hand_l, loc=<1.7616 -0.9753 3.3814>, rot=<-0.0397 0.0814 -0.069 -0.9935>}
 pinky_01_l {parent=pinky_metacarpal_l, loc=<7.1709 -0.4831 -0.1094>, rot=<-0.0654 -0.0035 0.2242 -0.9723>}
 pinky_02_l {parent=pinky_01_l, loc=<2.8274 0.0893 -0.1043>, rot=<0.0037 0.0011 0.2854 -0.9584>}
 pinky_03_l {parent=pinky_02_l, loc=<2.2046 0.0307 -0.0078>, rot=<-0.0002 0.0098 0.1445 -0.9895>}
 deform_wrist_dn_base_l {parent=hand_l, loc=<0 -2.2342 0>, rot=<0 0 0 -1>}
 deform_wrist_dn_l {parent=deform_wrist_dn_base_l, loc=<0 0 0>, rot=<0 0 0 1>}
 deform_wrist_up_base_l {parent=hand_l, loc=<0 2.2682 0>, rot=<0 0 0 -1>}
 deform_wrist_up_l {parent=deform_wrist_up_base_l, loc=<0 0 0>, rot=<0 0 0 1>}
 lowerarm_twist_02_l {parent=lowerarm_l, loc=<9.6199 0 0>, rot=<0 0 0 -1>}
 lowerarm_twist_01_l {parent=lowerarm_l, loc=<17.8549 0 0>, rot=<0 0 0 -1>}
 deform_elbow_in_base_l {parent=lowerarm_l, loc=<0.2112 5.259 0.8274>, rot=<0 0 0 -1>}
 deform_elbow_in_l {parent=deform_elbow_in_base_l, loc=<0 0 0>, rot=<0 0 0 1>}
 deform_elbow_out_base_l {parent=lowerarm_l, loc=<0 -4 0>, rot=<0 0 0 -1>}
 deform_elbow_out_l {parent=deform_elbow_out_base_l, loc=<0 0 0>, rot=<0 0 0 1>}
 upperarm_twist_01_l {parent=upperarm_l, loc=<0.0041 0 0>, rot=<0 0 0 -1>}
 deform_upperarm_bk_base_l {parent=upperarm_twist_01_l, loc=<1.1959 -7.2193 -1.8747>, rot=<0 0.0302 0 -0.9995>}
 deform_upperarm_bk_l {parent=deform_upperarm_bk_base_l, loc=<0 0 0>, rot=<0 0 0 1>}
 deform_upperarm_fr_base_l {parent=upperarm_twist_01_l, loc=<-0.2012 7.0506 -2.7798>, rot=<0 0.0302 0 -0.9995>}
 deform_upperarm_fr_l {parent=deform_upperarm_fr_base_l, loc=<0 0 0>, rot=<0 0 0 1>}
 deform_upperarm_up_base_l {parent=upperarm_twist_01_l, loc=<-3.3317 -1.1391 3.2151>, rot=<0 0.0173 0 -0.9999>}
 deform_upperarm_up_l {parent=deform_upperarm_up_base_l, loc=<0 0 0>, rot=<0 0 0 1>}
 upperarm_twist_02_l {parent=upperarm_l, loc=<13.7941 0 0>, rot=<0 0 0 -1>}
 Dyn_cape_01_l1 {parent=clavicle_l, loc=<11.2272 -4.8398 7.8274>, rot=<-0.1932 -0.1531 0.6291 -0.7372>}
 Dyn_cape_02_l {parent=Dyn_cape_01_l1, loc=<4.8063 13.7455 -4.0269>, rot=<-0.0241 -0.3721 0.0366 -0.9272>}
 Dyn_cape_03_l {parent=Dyn_cape_02_l, loc=<14.7238 0 -0.631>, rot=<0 -0.108 0 -0.9941>}
 Dyn_cape_04_l {parent=Dyn_cape_03_l, loc=<29.2611 0.275 -0.031>, rot=<0 -0.0372 0 -0.9993>}
 Dyn_cape_05_l {parent=Dyn_cape_04_l, loc=<25.8097 0 0>, rot=<0 0 0 1>}
 Dyn_cape_06_l {parent=Dyn_cape_05_l, loc=<22.8312 0 0>, rot=<0 0 0 1>}
 Dyn_cape_07_l {parent=Dyn_cape_06_l, loc=<14.1389 0 0>, rot=<0 0 0 1>}
 clavicle_r {parent=spine_05, loc=<4.3575 -0.1466 2.3223>, rot=<0.7289 0.0465 -0.6829 0.009>}
 upperarm_r {parent=clavicle_r, loc=<-15.5111 0 0>, rot=<0.0226 -0.3234 -0.0088 -0.9459>}
 lowerarm_r {parent=upperarm_r, loc=<-26.6514 0 0>, rot=<0.0391 -0.0099 -0.2453 -0.9686>}
 hand_r {parent=lowerarm_r, loc=<-22.3491 0 0>, rot=<-0.7081 -0.0507 0.0975 -0.6975>}
 thumb_01_r {parent=hand_r, loc=<-1.3118 0.3191 2.459>, rot=<0.6546 -0.4042 0.0664 -0.6354>}
 thumb_02_r {parent=thumb_01_r, loc=<-5.9336 0.0448 -0.0401>, rot=<-0.0037 0.0213 0.298 -0.9543>}
 thumb_03_r {parent=thumb_02_r, loc=<-2.2223 -0.3682 0.0907>, rot=<0.0118 -0.013 0.0066 -0.9998>}
 index_metacarpal_r {parent=hand_r, loc=<-1.6535 -0.0183 2.96>, rot=<0.0158 -0.0124 -0.0173 -0.9996>}
 index_01_r {parent=index_metacarpal_r, loc=<-8.3008 -0.0357 0.0524>, rot=<-0.0044 0.0485 0.0924 -0.9945>}
 index_02_r {parent=index_01_r, loc=<-3.3587 0.3004 -0.0452>, rot=<-0.002 0.0027 0.3225 -0.9466>}
 index_03_r {parent=index_02_r, loc=<-2.5578 0.123 0.0252>, rot=<-0.002 0.0067 0.2558 -0.9667>}
 middle_metacarpal_r {parent=hand_r, loc=<-2.1856 0.1725 0.5984>, rot=<0.006 0.0168 -0.0329 -0.9993>}
 middle_01_r {parent=middle_metacarpal_r, loc=<-7.3275 0.2447 -0.0129>, rot=<-0.0208 0.0042 0.0858 -0.9961>}
 middle_02_r {parent=middle_01_r, loc=<-4.0123 -0.1401 -0.1175>, rot=<-0.0218 0.0258 0.3512 -0.9357>}
 middle_03_r {parent=middle_02_r, loc=<-2.6077 -0.0534 -0.1373>, rot=<0.0082 0.0188 0.1708 -0.9851>}
 ring_metacarpal_r {parent=hand_r, loc=<-2.2074 0.3625 -1.3638>, rot=<-0.0284 0.0722 -0.0458 -0.9959>}
 ring_01_r {parent=ring_metacarpal_r, loc=<-7.0295 0.6197 0.1446>, rot=<-0.0489 -0.0252 0.1892 -0.9804>}
 ring_02_r {parent=ring_01_r, loc=<-3.6174 -0.4647 0.1027>, rot=<0 0 0.2806 -0.9598>}
 ring_03_r {parent=ring_02_r, loc=<-2.4152 -0.114 -0.0199>, rot=<0 0 0.1189 -0.9929>}
 pinky_metacarpal_r {parent=hand_r, loc=<-1.7616 0.9753 -3.3814>, rot=<-0.0397 0.0814 -0.069 -0.9935>}
 pinky_01_r {parent=pinky_metacarpal_r, loc=<-7.1709 0.4831 0.1094>, rot=<-0.0654 -0.0035 0.2242 -0.9723>}
 pinky_02_r {parent=pinky_01_r, loc=<-2.8274 -0.0893 0.1043>, rot=<0.0037 0.0011 0.2854 -0.9584>}
 pinky_03_r {parent=pinky_02_r, loc=<-2.2046 -0.0307 0.0078>, rot=<-0.0002 0.0098 0.1445 -0.9895>}
 deform_wrist_dn_base_r {parent=hand_r, loc=<-0.0004 2.2345 -0.0001>, rot=<0 0 0 -1>}
 deform_wrist_dn_r {parent=deform_wrist_dn_base_r, loc=<0 0 0>, rot=<0 0 0 1>}
 deform_wrist_up_base_r {parent=hand_r, loc=<0.0003 -2.2684 0.0001>, rot=<0 0 0 -1>}
 deform_wrist_up_r {parent=deform_wrist_up_base_r, loc=<0 0 0>, rot=<0 0 0 1>}
 lowerarm_twist_02_r {parent=lowerarm_r, loc=<-9.6199 0 0>, rot=<0 0 0 -1>}
 lowerarm_twist_01_r {parent=lowerarm_r, loc=<-17.8549 0 0>, rot=<0 0 0 -1>}
 deform_elbow_in_base_r {parent=lowerarm_r, loc=<-0.2111 -5.2591 -0.8275>, rot=<0 0 0 -1>}
 deform_elbow_in_r {parent=deform_elbow_in_base_r, loc=<0 0 0>, rot=<0 0 0 1>}
 deform_elbow_out_base_r {parent=lowerarm_r, loc=<0 4 -0.0001>, rot=<0 0 0 -1>}
 deform_elbow_out_r {parent=deform_elbow_out_base_r, loc=<0 0 0>, rot=<0 0 0 1>}
 upperarm_twist_01_r {parent=upperarm_r, loc=<-0.0041 0 0>, rot=<0 0 0 -1>}
 deform_upperarm_bk_base_r {parent=upperarm_twist_01_r, loc=<-1.0814 7.2192 1.9455>, rot=<0 0 0 1>}
 deform_upperarm_bk_r {parent=deform_upperarm_bk_base_r, loc=<0 0 0>, rot=<0 0 0 1>}
 deform_upperarm_fr_base_r {parent=upperarm_twist_01_r, loc=<0.3689 -7.0506 2.7652>, rot=<0 0 0 1>}
 deform_upperarm_fr_r {parent=deform_upperarm_fr_base_r, loc=<0 0 0>, rot=<0 0 0 1>}
 deform_upperarm_up_base_r {parent=upperarm_twist_01_r, loc=<3.2214 1.1391 -3.3308>, rot=<0 0 0 1>}
 deform_upperarm_up_r {parent=deform_upperarm_up_base_r, loc=<0 0 0>, rot=<0 0 0 -1>}
 upperarm_twist_02_r {parent=upperarm_r, loc=<-13.7941 0 0>, rot=<0 0 0 -1>}
 base_shoulderpad_side_00_rt {parent=upperarm_r, loc=<0.6145 1.2559 -5.1517>, rot=<-0.0043 0.7018 0.7121 -0.021>}
 dyn_shoulderpad_side_00_rt {parent=base_shoulderpad_side_00_rt, loc=<0 0 0>, rot=<0 0 0 1>}
 Dyn_cape_01_r1 {parent=clavicle_r, loc=<-11.2272 4.8398 -7.8274>, rot=<-0.6291 -0.7372 -0.1932 0.1531>}
 Dyn_cape_02_r {parent=Dyn_cape_01_r1, loc=<4.8063 -13.7455 -4.0269>, rot=<0.0241 -0.3721 -0.0366 -0.9272>}
 Dyn_cape_03_r {parent=Dyn_cape_02_r, loc=<14.7239 0 -0.6311>, rot=<0 -0.1015 0 -0.9948>}
 Dyn_cape_04_r {parent=Dyn_cape_03_r, loc=<29.2614 -0.275 -0.0311>, rot=<0 0 0 1>}
 Dyn_cape_05_r {parent=Dyn_cape_04_r, loc=<25.8097 0 0>, rot=<0 0 0 1>}
 Dyn_cape_06_r {parent=Dyn_cape_05_r, loc=<22.8312 0 0>, rot=<0 0 0 1>}
 Dyn_cape_07_r {parent=Dyn_cape_06_r, loc=<14.1389 0 0>, rot=<0 0 0 1>}
 base_fur_front_00_rt {parent=spine_05, loc=<12.5832 4.6145 16.521>, rot=<0.7071 0.0064 0.7053 -0.0502>}
 dyn_fur_front_00_rt {parent=base_fur_front_00_rt, loc=<0 0 0>, rot=<0 0 0 1>}
 base_fur_front_00_lf {parent=spine_05, loc=<12.5832 4.6145 -16.521>, rot=<0.0064 -0.7071 -0.0502 -0.7053>}
 dyn_fur_front_00_lf {parent=base_fur_front_00_lf, loc=<0 0 0>, rot=<0 0 0 1>}
 deform_pec_aim_base_l {parent=spine_04, loc=<7.8732 12.6388 -2.4123>, rot=<0.0102 -0.4602 0.3979 -0.7936>}
 deform_pec_aim_l {parent=deform_pec_aim_base_l, loc=<0 0 0>, rot=<0 0 0 1>}
 deform_pec_base_l {parent=deform_pec_aim_l, loc=<10.766 0 0>, rot=<0 0 0 1>}
 deform_pec_l {parent=deform_pec_base_l, loc=<0.055 0 0>, rot=<0 0 0 1>}
 deform_pec_aim_base_r {parent=spine_04, loc=<7.8727 12.6388 2.4123>, rot=<-0.4602 -0.0102 0.7936 0.3979>}
 deform_pec_aim_r {parent=deform_pec_aim_base_r, loc=<0 0 0>, rot=<0 0 0 1>}
 deform_pec_base_r {parent=deform_pec_aim_r, loc=<-10.7662 -0.0006 -0.0007>, rot=<0 0 0 1>}
 deform_pec_r {parent=deform_pec_base_r, loc=<0 0 0>, rot=<0 0 0 1>}
 thigh_l {parent=pelvis, loc=<-2.5693 -0.0109 -7.7501>, rot=<0.0012 0.0572 0.0155 -0.9982>}
 calf_l {parent=thigh_l, loc=<-39.0484 0 0>, rot=<0.0006 -0.0161 -0.0391 -0.9991>}
 foot_l {parent=calf_l, loc=<-42.561 0.1077 0.4292>, rot=<0.0418 -0.0455 0.0343 -0.9975>}
 ball_l {parent=foot_l, loc=<-6.1599 13.8264 0.6806>, rot=<0.0001 -0.0001 -0.7063 -0.7079>}
 calf_twist_02_l {parent=calf_l, loc=<-18.8064 -0.0967 0.0397>, rot=<0.0237 0.0095 -0.0035 -0.9997>}
 calf_twist_01_l {parent=calf_l, loc=<-35.813 -0.0967 0.0397>, rot=<0.0237 0.0095 -0.0035 -0.9997>}
 deform_knee_in_base_l {parent=calf_l, loc=<1.298 -5.2957 -0.0496>, rot=<0 0 0 -1>}
 deform_knee_in_l {parent=deform_knee_in_base_l, loc=<0 0 0>, rot=<0 0 0 1>}
 deform_knee_out_base_l {parent=calf_l, loc=<0 4 0>, rot=<0 0 0 -1>}
 deform_knee_out_l {parent=deform_knee_out_base_l, loc=<0 0 0>, rot=<0 0 0 1>}
 thigh_twist_01_l {parent=thigh_l, loc=<-21.6303 0 0>, rot=<0.0013 0 0 -1>}
 deform_groin_base_l {parent=thigh_l, loc=<-2.1908 8.1142 -1.7442>, rot=<0 0 0 -1>}
 deform_groin_l {parent=deform_groin_base_l, loc=<0 0 0>, rot=<0 0 0 1>}
 thigh_r {parent=pelvis, loc=<-2.5693 -0.0109 7.7501>, rot=<0.0572 -0.0012 0.9982 0.0155>}
 calf_r {parent=thigh_r, loc=<39.0484 0 0>, rot=<0.0006 -0.0161 -0.0391 -0.9991>}
 foot_r {parent=calf_r, loc=<42.561 -0.1077 -0.4292>, rot=<0.0418 -0.0455 0.0343 -0.9975>}
 ball_r {parent=foot_r, loc=<6.1599 -13.8264 -0.6806>, rot=<0.0001 -0.0001 -0.7063 -0.7079>}
 calf_twist_02_r {parent=calf_r, loc=<18.8064 0.0967 -0.0397>, rot=<0.0237 0.0095 -0.0035 -0.9997>}
 calf_twist_01_r {parent=calf_r, loc=<35.813 0.0967 -0.0397>, rot=<0.0237 0.0095 -0.0035 -0.9997>}
 deform_knee_in_base_r {parent=calf_r, loc=<-1.298 5.2957 0.0496>, rot=<0 0 0 -1>}
 deform_knee_in_r {parent=deform_knee_in_base_r, loc=<0 0 0>, rot=<0 0 0 1>}
 deform_knee_out_base_r {parent=calf_r, loc=<0 -4 0>, rot=<0 0 0 -1>}
 deform_knee_out_r {parent=deform_knee_out_base_r, loc=<0 0 0>, rot=<0 0 0 1>}
 thigh_twist_01_r {parent=thigh_r, loc=<21.6303 0 0>, rot=<0.0013 0 0 -1>}
 deform_groin_base_r {parent=thigh_r, loc=<2.1908 -8.1142 1.7442>, rot=<0 0 0 -1>}
 deform_groin_r {parent=deform_groin_base_r, loc=<0 0 0>, rot=<0 0 0 1>}
 deform_glute_base_l {parent=pelvis, loc=<-8.8064 -8.8222 -8.4287>, rot=<0.0275 0.0501 -0.4472 -0.8926>}
 deform_glute_l {parent=deform_glute_base_l, loc=<0 0 0>, rot=<0 0 0 1>}
 deform_glute_base_r {parent=pelvis, loc=<-8.8062 -8.8222 8.4287>, rot=<0.0501 -0.0275 0.8926 -0.4472>}
 deform_glute_r {parent=deform_glute_base_r, loc=<0 0 0>, rot=<0 0 0 1>}
 base_belt_side_00_lf {parent=pelvis, loc=<-11.0466 1.7208 -18.5157>, rot=<-0.0349 0.5519 0.832 0.0446>}
 dyn_belt_side_00_lf {parent=base_belt_side_00_lf, loc=<0 0 0>, rot=<0 0 0 1>}
 dyn_belt_side_01_lf {parent=dyn_belt_side_00_lf, loc=<10 0 0>, rot=<-0.0012 -0.0349 -0.0349 -0.9988>}
 dyn_belt_side_02_lf {parent=dyn_belt_side_01_lf, loc=<10 0 0>, rot=<0 0 0 1>}
 base_belt_side_00_rt {parent=pelvis, loc=<-8.3204 4.7483 18.1488>, rot=<0.0403 -0.6418 0.7654 0.0244>}
 dyn_belt_side_00_rt {parent=base_belt_side_00_rt, loc=<0 0 0>, rot=<0 0 0 1>}
 dyn_belt_side_01_rt {parent=dyn_belt_side_00_rt, loc=<12.3769 0 0>, rot=<0 0 0 1>}
 dyn_belt_side_02_rt {parent=dyn_belt_side_01_rt, loc=<12.3769 0 0>, rot=<0 0 0 1>}
 base_belt_front_00_rt {parent=pelvis, loc=<1.7445 11.4453 -8.8036>, rot=<-0.011 0.0699 0.9973 0.0183>}
 dyn_belt_front_00_rt {parent=base_belt_front_00_rt, loc=<0 0 0>, rot=<0 0 0 1>}
 dyn_belt_front_01_rt {parent=dyn_belt_front_00_rt, loc=<8.0544 0 0>, rot=<0 0 -0.0349 -0.9994>}
 dyn_belt_front_02_rt {parent=dyn_belt_front_01_rt, loc=<9.3814 0 0>, rot=<0 0 0 1>}
]
variables [
 Aim {type=bool, default=False}
 GoldenGloryEmote {type=bool, default=False}
 IsCrouching {type=bool, default=False}
]
