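# Physics asset for the M_MED_SolveSolid male medium player body cosmetic, apparently built for dangling belt pieces.
# Capsules on the pelvis, thigh twist bones, calves and feet form a lower-body shape with locked angular constraints, with both calves tied directly to the pelvis.
# Three belt chains (front right, side left, side right) each hang three dyn bones from a tiny base anchor locked to the pelvis, with swing limits set per link.
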PhysicsAsset /BRCosmetics/Characters/Player/Male/Medium/Bodies/M_MED_SolveSolid/Meshes/M_MED_SolveSolid_Physics
bodies {
 base_belt_front_00_rt [{shapeType=Capsule, length=0.505, radius=0.505, shapeName=base_belt_front_00_rt_capsule, rotation=<rot 0 -90 0>}]
 base_belt_side_00_lf [{shapeType=Capsule, length=0.505, radius=0.505, shapeName=base_belt_side_00_lf_capsule, rotation=<rot 0 -90 0>}]
 base_belt_side_00_rt [{shapeType=Capsule, length=0.505, radius=0.505, shapeName=base_belt_side_00_rt_capsule, rotation=<rot 0 -90 0>}]
 calf_l [{shapeType=Capsule, length=21.2372, radius=8.7837, center=<-19.6808 -2.1081 -0.3219>, rotation=<rot 0 -88.3524 -88.1356>}]
 calf_r [{shapeType=Capsule, length=21.4204, radius=9.0395, center=<19.5229 2.0074 0.3708>, rotation=<rot 0 -88.1308 -88.6549>}]
 dyn_belt_front_00_rt [{shapeType=Capsule, length=4.2548, radius=1.6604, shapeName=dyn_belt_front_00_rt_capsule, center=<4.1311 -1 0>, rotation=<rot 90 0 0>}]
 dyn_belt_front_01_rt [{
   shapeType=Capsule
   length=5.8827
   radius=1.5277
   shapeName=dyn_belt_front_01_rt_capsule
   center=<4.4326 -0.75 0>
   rotation=<rot 90 -0.0001 0>
  }]
 dyn_belt_front_02_rt [{shapeType=Capsule, length=8.5085, radius=1, shapeName=dyn_belt_front_02_rt_capsule, center=<2.4744 -0.2 0.2>, rotation=<rot 90 -0.0002 0>}]
 dyn_belt_side_00_lf [{shapeType=Capsule, length=3, radius=2, shapeName=dyn_belt_side_00_lf_capsule, center=<4 -1 0>}]
 dyn_belt_side_00_rt [{shapeType=Capsule, length=7.2469, radius=4.2597, shapeName=dyn_belt_side_00_rt_capsule, center=<6.9104 -2.5 0>, rotation=<rot 90 0 0>}]
 dyn_belt_side_01_lf [{shapeType=Capsule, length=5.5685, radius=1.5719, shapeName=dyn_belt_side_01_lf_capsule, center=<4.8803 -0.3411 0>, rotation=<rot 90 0 0>}]
 dyn_belt_side_01_rt [{
   shapeType=Capsule
   length=5.8938
   radius=4.2639
   shapeName=dyn_belt_side_01_rt_capsule
   center=<5.7274 -1 -1.5>
   rotation=<rot -0.0001 -0.0001 -15>
  }]
 dyn_belt_side_02_lf [{shapeType=Capsule, length=4.6359, radius=2.5437, shapeName=dyn_belt_side_02_lf_capsule, center=<1.6078 0 0>}]
 dyn_belt_side_02_rt [{
   shapeType=Capsule
   length=7.3643
   radius=5.001
   shapeName=dyn_belt_side_02_rt_capsule
   center=<-0.8059 -1.9699 -1.3043>
   rotation=<rot -17.9001 -90 6.7431>
  }]
 foot_l [{shapeType=Capsule, length=13.9693, radius=8.9346, center=<-3.004 7.8933 -0.325>, rotation=<rot 0 10.1526 -88.6818>}]
 foot_r [{shapeType=Capsule, length=14.0517, radius=8.9575, center=<3.034 -7.8245 0.331>, rotation=<rot 0 10.4501 -88.516>}]
 pelvis [{shapeType=Capsule, length=18.1922, radius=11.0082, center=<-5.2368 -0.2528 0>}]
 thigh_twist_01_l [{shapeType=Capsule, length=24.6397, radius=10.5965, center=<0.065 -2.4105 1.7726>, rotation=<rot 90 4.9999 0>}]
 thigh_twist_01_r [{shapeType=Capsule, length=24.6397, radius=10.5965, center=<0 2.4273 -1.7726>, rotation=<rot 90 180 0>}]
}
constraints [
 {
  bone1Name=calf_l
  bone2Name=pelvis
  swing1LimitDegrees=45
  swing1Motion=Locked
  swing2LimitDegrees=45
  swing2Motion=Locked
  twistLimitDegrees=45
  twistMotion=Locked
 }
 {
  bone1Name=foot_l
  bone2Name=calf_l
  swing1LimitDegrees=45
  swing1Motion=Locked
  swing2LimitDegrees=45
  swing2Motion=Locked
  twistLimitDegrees=45
  twistMotion=Locked
 }
 {
  bone1Name=thigh_twist_01_l
  bone2Name=pelvis
  swing1LimitDegrees=45
  swing1Motion=Locked
  swing2LimitDegrees=45
  swing2Motion=Locked
  twistLimitDegrees=45
  twistMotion=Locked
 }
 {
  bone1Name=calf_r
  bone2Name=pelvis
  swing1LimitDegrees=45
  swing1Motion=Locked
  swing2LimitDegrees=45
  swing2Motion=Locked
  twistLimitDegrees=45
  twistMotion=Locked
 }
 {
  bone1Name=foot_r
  bone2Name=calf_r
  swing1LimitDegrees=45
  swing1Motion=Locked
  swing2LimitDegrees=45
  swing2Motion=Locked
  twistLimitDegrees=45
  twistMotion=Locked
 }
 {
  bone1Name=thigh_twist_01_r
  bone2Name=pelvis
  swing1LimitDegrees=45
  swing1Motion=Locked
  swing2LimitDegrees=45
  swing2Motion=Locked
  twistLimitDegrees=45
  twistMotion=Locked
 }
 {
  bone1Name=base_belt_front_00_rt
  bone2Name=pelvis
  swing1LimitDegrees=45
  swing1Motion=Locked
  swing2LimitDegrees=45
  swing2Motion=Locked
  twistLimitDegrees=45
  twistMotion=Locked
 }
 {bone1Name=dyn_belt_front_00_rt, bone2Name=base_belt_front_00_rt, swing1LimitDegrees=20, swing2LimitDegrees=5, twistLimitDegrees=0}
 {bone1Name=dyn_belt_front_01_rt, bone2Name=dyn_belt_front_00_rt, swing1LimitDegrees=30, swing2LimitDegrees=5, twistLimitDegrees=0}
 {bone1Name=dyn_belt_front_02_rt, bone2Name=dyn_belt_front_01_rt, swing1LimitDegrees=25, swing2LimitDegrees=5, twistLimitDegrees=0}
 {
  bone1Name=base_belt_side_00_lf
  bone2Name=pelvis
  swing1LimitDegrees=45
  swing1Motion=Locked
  swing2LimitDegrees=45
  swing2Motion=Locked
  twistLimitDegrees=45
  twistMotion=Locked
 }
 {bone1Name=dyn_belt_side_00_lf, bone2Name=base_belt_side_00_lf, swing1LimitDegrees=15, swing2LimitDegrees=15, twistLimitDegrees=0}
 {bone1Name=dyn_belt_side_01_lf, bone2Name=dyn_belt_side_00_lf, swing1LimitDegrees=15, swing2LimitDegrees=15, twistLimitDegrees=0}
 {bone1Name=dyn_belt_side_02_lf, bone2Name=dyn_belt_side_01_lf, swing1LimitDegrees=25, swing2LimitDegrees=25, twistLimitDegrees=0}
 {
  bone1Name=base_belt_side_00_rt
  bone2Name=pelvis
  swing1LimitDegrees=45
  swing1Motion=Locked
  swing2LimitDegrees=45
  swing2Motion=Locked
  twistLimitDegrees=45
  twistMotion=Locked
 }
 {bone1Name=dyn_belt_side_00_rt, bone2Name=base_belt_side_00_rt, swing1LimitDegrees=15, swing2LimitDegrees=15, twistLimitDegrees=0}
 {bone1Name=dyn_belt_side_01_rt, bone2Name=dyn_belt_side_00_rt, swing1LimitDegrees=15, swing2LimitDegrees=15, twistLimitDegrees=0}
 {bone1Name=dyn_belt_side_02_rt, bone2Name=dyn_belt_side_01_rt, swing1LimitDegrees=25, swing2LimitDegrees=25, twistLimitDegrees=0}
]
