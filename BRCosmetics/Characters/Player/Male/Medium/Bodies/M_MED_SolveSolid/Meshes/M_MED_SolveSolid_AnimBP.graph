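# Post-process style AnimBP for the M_MED_SolveSolid player body, adding Control Rig corrections and cape, belt and fur dynamics to the pose copied from the parent mesh.
# With long backbling, cloth is disabled and a nudged backpack-cape RigidBody takes over; otherwise the cape uses cloth, or RigidBody when cloth is off.
# Another RigidBody and AnimDynamics follow.
# The EventGraph drives gravity, cloth, cape-fix and GoldenGlory emote flags.
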AnimBlueprint /BRCosmetics/Characters/Player/Male/Medium/Bodies/M_MED_SolveSolid/Meshes/M_MED_SolveSolid_AnimBP
target_skeleton=&/M_MED_SolveSolid_Skeleton
graphs [
 AnimGraph {nodes [
   Output Pose(Result=Use cached pose 'Dynamics'()) {layerGroup=None}
   n1 = Control Rig!(
    Source=Copy Pose From Mesh() {bUseAttachedParent=true}
    Alpha
    IsCrouching:bool=Get bIsCrouching()
    GoldenGloryEmote:bool=Get GoldenGloryEmote() -> (GoldenGloryEmote:bool)
   ) -> (Pose) {controlRigAssetReference {blueprintRigClass=&/M_MED_SolveSolid_CtrlRig.M_MED_SolveSolid_CtrlRig_C}}
   Save cached pose 'ControlRig_Node'(Pose=@n1) {cacheName=ControlRig_Node}
   n2 = "Transform (Modify) Bone - Bone: attach_backpack"(
    ComponentPose=Local To Component(LocalPose=Use cached pose 'ControlRig_Node'())
    Alpha=1
    Translation="0.000,-9.673,0.322"
    Rotation="0.000,-0.000,5.000"
    Scale="1,1,1"
   ) {
    rotationMode=BMM_Additive
    rotationSpace=BCS_BoneSpace
    translationMode=BMM_Additive
    translationSpace=BCS_BoneSpace
    boneToModify {boneName=attach_backpack}
    rotation=<rot 0 0 4.9999>
    translation {y=-9.673, z=0.322}
   }
   n3 = "Transform (Modify) Bone - Bone: attach_cape"(ComponentPose=@n2, Alpha=1, Translation="-0.000,-7.505,-0.000", Rotation="0,0,0", Scale="1,1,1") {translationMode=BMM_Additive, translationSpace=BCS_BoneSpace, boneToModify {boneName=attach_cape}, translation {y=-7.505}}
   Save cached pose 'PreDynamics'(Pose=Component To Local(ComponentPose=@n3)) {cacheName=PreDynamics}
   n4 = Get CapeFIx() -> (CapeFIx:bool)
   n5 = "Transform (Modify) Bone - Bone: Dyn_cape_01_l1"!(
    ComponentPose=Local To Component(LocalPose=Use cached pose 'PreDynamics'())
    bAlphaBoolEnabled:bool=@n4
    Translation="0,0,-3"
    Rotation="-15,0,0"
    Scale="1,1,1"
   ) -> (Pose) {
    alphaInputType=Bool
    rotationMode=BMM_Additive
    rotationSpace=BCS_BoneSpace
    translationMode=BMM_Additive
    translationSpace=BCS_BoneSpace
    boneToModify {boneName=Dyn_cape_01_l1}
    rotation {pitch=-15}
    translation {z=-3}
   }
   n6 = "Transform (Modify) Bone - Bone: Dyn_cape_01_r1"!(ComponentPose=@n5, bAlphaBoolEnabled:bool=@n4, Translation="0,0,-3", Rotation="15,0,0", Scale="1,1,1") -> (Pose) {
    alphaInputType=Bool
    rotationMode=BMM_Additive
    rotationSpace=BCS_BoneSpace
    translationMode=BMM_Additive
    translationSpace=BCS_BoneSpace
    boneToModify {boneName=Dyn_cape_01_r1}
    rotation {pitch=15}
    translation {z=-3}
   }
   n7 = "Transform (Modify) Bone - Bone: Dyn_cape_mid_01_r_MSTR"!(ComponentPose=@n6, bAlphaBoolEnabled:bool=@n4, Translation="0,0,-3", Rotation="-15,0,0", Scale="1,1,1") -> (Pose) {
    alphaInputType=Bool
    rotationMode=BMM_Additive
    rotationSpace=BCS_BoneSpace
    translationMode=BMM_Additive
    translationSpace=BCS_BoneSpace
    boneToModify {boneName=Dyn_cape_mid_01_r_MSTR}
    rotation {pitch=-15}
    translation {z=-3}
   }
   n8 = "Transform (Modify) Bone - Bone: Dyn_cape_mid_01_l_MSTR"!(ComponentPose=@n7, bAlphaBoolEnabled:bool=@n4, Translation="0,0,-3", Rotation="-15,0,0", Scale="1,1,1") -> (Pose) {
    alphaInputType=Bool
    rotationMode=BMM_Additive
    rotationSpace=BCS_BoneSpace
    translationMode=BMM_Additive
    translationSpace=BCS_BoneSpace
    boneToModify {boneName=Dyn_cape_mid_01_l_MSTR}
    rotation {pitch=-15}
    translation {z=-3}
   }
   n9 = RigidBody!(ComponentPose=@n8, Alpha, OverrideWorldGravity:Vector="0,0,0", ExternalForce) -> (Pose) {
    bOverrideWorldGravity=true
    overridePhysicsAsset=&/M_MED_SolveSolid_Cape_Physics
    simulationSpace=BaseBoneSpace
    baseBoneRef {boneName=spine_05}
    simSpaceSettings {
     dampingAlpha=0
     maxAngularAcceleration=40
     maxAngularVelocity=8
     maxLinearAcceleration=3500
     maxLinearVelocity=800
     velocityScaleZ=0.7
     worldAlpha=1
     externalLinearDragV=<0.2 0.2 0.2>
     externalLinearVelocity=<0.2 0.2 0.2>
    }
   }
   n10 = Blend Poses by bool(BlendPose_0=Use cached pose 'PreDynamics'(), BlendPose_1=Component To Local(ComponentPose=@n9), bActiveValue=Get bClothEnabled())
   n11 = "Transform (Modify) Bone - Bone: Dyn_cape_01_l1"(
    ComponentPose=Local To Component(LocalPose=Use cached pose 'PreDynamics'())
    bAlphaBoolEnabled=True
    Translation="0,0,-3"
    Rotation="-7,0,0"
    Scale="1,1,1"
   ) {
    alphaInputType=Bool
    rotationMode=BMM_Additive
    rotationSpace=BCS_BoneSpace
    translationMode=BMM_Additive
    translationSpace=BCS_BoneSpace
    boneToModify {boneName=Dyn_cape_01_l1}
    rotation {pitch=-7}
    translation {z=-3}
   }
   n12 = "Transform (Modify) Bone - Bone: Dyn_cape_01_r1"(ComponentPose=@n11, bAlphaBoolEnabled=True, Translation="0,0,-3", Rotation="7,0,0", Scale="1,1,1") {
    alphaInputType=Bool
    rotationMode=BMM_Additive
    rotationSpace=BCS_BoneSpace
    translationMode=BMM_Additive
    translationSpace=BCS_BoneSpace
    boneToModify {boneName=Dyn_cape_01_r1}
    rotation {pitch=7}
    translation {z=-3}
   }
   n13 = "Transform (Modify) Bone - Bone: Dyn_cape_mid_01_r_MSTR"(ComponentPose=@n12, bAlphaBoolEnabled=True, Translation="0,0,-3", Rotation="-7,0,0", Scale="1,1,1") {
    alphaInputType=Bool
    rotationMode=BMM_Additive
    rotationSpace=BCS_BoneSpace
    translationMode=BMM_Additive
    translationSpace=BCS_BoneSpace
    boneToModify {boneName=Dyn_cape_mid_01_r_MSTR}
    rotation {pitch=-7}
    translation {z=-3}
   }
   n14 = "Transform (Modify) Bone - Bone: Dyn_cape_mid_01_l_MSTR"(ComponentPose=@n13, bAlphaBoolEnabled=True, Translation="0,0,-3", Rotation="-7,0,0", Scale="1,1,1") {
    alphaInputType=Bool
    rotationMode=BMM_Additive
    rotationSpace=BCS_BoneSpace
    translationMode=BMM_Additive
    translationSpace=BCS_BoneSpace
    boneToModify {boneName=Dyn_cape_mid_01_l_MSTR}
    rotation {pitch=-7}
    translation {z=-3}
   }
   n15 = RigidBody(ComponentPose=@n14, OverrideWorldGravity="0,0,0") {
    bOverrideWorldGravity=true
    overridePhysicsAsset=&/M_MED_SolveSolid_BPCape_Physics
    simulationSpace=BaseBoneSpace
    baseBoneRef {boneName=spine_05}
    simSpaceSettings {
     dampingAlpha=0
     maxAngularAcceleration=25
     maxAngularVelocity=5
     maxLinearAcceleration=2000
     maxLinearVelocity=400
     worldAlpha=1
     externalLinearDragV=<0.3 0.2 0.2>
     externalLinearVelocity=<0.3 0.2 0.2>
    }
   }
   n16 = Component To Local(ComponentPose=@n15)
   n17 = Get IsWearingBackbling() -> (IsWearingBackbling:bool)
   n18 = RigidBody(
    ComponentPose=Local To Component(LocalPose=Blend Poses by bool(BlendPose_0=@n16, BlendPose_1=@n10, bActiveValue=@n17))
    OverrideWorldGravity="0,0,-980"
   ) {
    bOverrideWorldGravity=true
    overridePhysicsAsset=&/M_MED_SolveSolid_Physics
    simulationSpace=BaseBoneSpace
    baseBoneRef {boneName=thigh_l}
    componentLinearAccScale {z=1}
    overrideWorldGravity {z=-980}
    simSpaceSettings {
     dampingAlpha=0
     maxAngularAcceleration=20
     maxAngularVelocity=4
     maxLinearAcceleration=3500
     maxLinearVelocity=700
     velocityScaleZ=0.6
     worldAlpha=1
     externalLinearDragV=<0.2 0.2 0.2>
     externalLinearVelocity=<0.2 0.2 0.2>
    }
   }
   n19 = Anim Dynamics - Bone: dyn_belt_side_00_rt!(ComponentPose=@n18, Alpha=1, GravityOverride:Vector="0,0,-980", ExternalForce="0,0,0") -> (Pose) {
    angularDampingOverride=0.999
    angularSpringConstant=10000
    bAngularSpring=true
    bOverrideAngularDamping=true
    bUseGravityOverride=true
    bUsePlanarLimit=false
    gravityScale=0.35
    simulationSpace=BoneRelative
    boundBone {boneName=dyn_belt_side_00_rt}
    componentAppliedLinearAccClamp=<1000 1000 1000>
    componentLinearAccScale=<1 1 1>
    componentLinearVelScale=<1 1 1>
    gravityOverride {z=-980}
    physicsBodyDefinitions [{
      boundBone {boneName=dyn_belt_side_00_rt}
      boxExtents=<20 30 30>
      constraintSetup {angularLimitsMax=<15 20 0>, angularLimitsMin=<-15 -20 -15>, angularTarget=<1 0 0>}
      localJointOffset=<0 0 0>
     }]
    relativeSpaceBone {boneName=thigh_r}
   }
   n20 = Anim Dynamics - Bone: dyn_belt_side_01_rt!(ComponentPose=@n19, Alpha=1, GravityOverride:Vector="0,0,-980", ExternalForce="0,0,0") -> (Pose) {
    angularDampingOverride=0.999
    angularSpringConstant=10000
    bAngularSpring=true
    bOverrideAngularDamping=true
    bUseGravityOverride=true
    bUsePlanarLimit=false
    gravityScale=0.35
    simulationSpace=BoneRelative
    boundBone {boneName=dyn_belt_side_01_rt}
    componentAppliedLinearAccClamp=<1000 1000 1000>
    componentLinearAccScale=<1 1 1>
    componentLinearVelScale=<1 1 1>
    gravityOverride {z=-980}
    physicsBodyDefinitions [{
      boundBone {boneName=dyn_belt_side_01_rt}
      boxExtents=<30 30 30>
      constraintSetup {angularLimitsMax=<15 25 0>, angularLimitsMin=<-15 -25 -25>, angularTarget=<1 0 0>}
      localJointOffset=<4 0 0>
     }]
    relativeSpaceBone {boneName=thigh_r}
   }
   n21 = Anim Dynamics - Bone: dyn_belt_side_02_rt!(ComponentPose=@n20, Alpha=1, GravityOverride:Vector="0,0,-980", ExternalForce="0,0,0") -> (Pose) {
    angularDampingOverride=0.999
    angularSpringConstant=10000
    bAngularSpring=true
    bOverrideAngularDamping=true
    bUseGravityOverride=true
    bUsePlanarLimit=false
    gravityScale=0.35
    simulationSpace=BoneRelative
    boundBone {boneName=dyn_belt_side_02_rt}
    componentAppliedLinearAccClamp=<1000 1000 1000>
    componentLinearAccScale=<1 1 1>
    componentLinearVelScale=<1 1 1>
    gravityOverride {z=-980}
    physicsBodyDefinitions [{
      boundBone {boneName=dyn_belt_side_02_rt}
      boxExtents=<20 20 30>
      constraintSetup {angularLimitsMax=<15 30 0>, angularLimitsMin=<-15 -30 -30>, angularTarget=<1 0 0>}
      localJointOffset=<4 0 0>
     }]
    relativeSpaceBone {boneName=thigh_r}
   }
   n22 = Anim Dynamics - Bone: dyn_fur_back_00_ct!(ComponentPose=@n21, Alpha=1, GravityOverride:Vector="0,0,-980", ExternalForce="0,0,0") -> (Pose) {
    angularDampingOverride=1
    angularSpringConstant=10000
    bAngularSpring=true
    bOverrideAngularDamping=true
    bUseGravityOverride=true
    gravityScale=0.35
    simulationSpace=BoneRelative
    boundBone {boneName=dyn_fur_back_00_ct}
    componentLinearAccScale=<1 1 1>
    componentLinearVelScale=<1 1 1>
    gravityOverride {z=-980}
    physicsBodyDefinitions [{
      boundBone {boneName=dyn_fur_back_00_ct}
      boxExtents=<20 30 30>
      constraintSetup {angularLimitsMax=<8 0 8>, angularLimitsMin=<-8 0 -8>, angularTarget=<1 0 0>}
      localJointOffset=<0 0 0>
     }]
    relativeSpaceBone {boneName=root}
   }
   n23 = Anim Dynamics - Bone: dyn_fur_back_00_lf!(ComponentPose=@n22, Alpha=1, GravityOverride:Vector="0,0,-980", ExternalForce="0,0,0") -> (Pose) {
    angularDampingOverride=1
    angularSpringConstant=10000
    bAngularSpring=true
    bOverrideAngularDamping=true
    bUseGravityOverride=true
    gravityScale=0.35
    simulationSpace=BoneRelative
    boundBone {boneName=dyn_fur_back_00_lf}
    componentLinearAccScale=<1 1 1>
    componentLinearVelScale=<1 1 1>
    gravityOverride {z=-980}
    physicsBodyDefinitions [{
      boundBone {boneName=dyn_fur_back_00_lf}
      boxExtents=<20 30 30>
      constraintSetup {angularLimitsMax=<8 0 8>, angularLimitsMin=<-8 0 -8>, angularTarget=<1 0 0>}
      localJointOffset=<0 0 0>
     }]
    relativeSpaceBone {boneName=root}
   }
   n24 = Anim Dynamics - Bone: dyn_fur_back_00_rt!(ComponentPose=@n23, Alpha=1, GravityOverride:Vector="0,0,-980", ExternalForce="0,0,0") -> (Pose) {
    angularDampingOverride=1
    angularSpringConstant=10000
    bAngularSpring=true
    bOverrideAngularDamping=true
    bUseGravityOverride=true
    gravityScale=0.35
    simulationSpace=BoneRelative
    boundBone {boneName=dyn_fur_back_00_rt}
    componentLinearAccScale=<1 1 1>
    componentLinearVelScale=<1 1 1>
    gravityOverride {z=-980}
    physicsBodyDefinitions [{
      boundBone {boneName=dyn_fur_back_00_rt}
      boxExtents=<20 30 30>
      constraintSetup {angularLimitsMax=<8 0 8>, angularLimitsMin=<-8 0 -8>, angularTarget=<1 0 0>}
      localJointOffset=<0 0 0>
     }]
    relativeSpaceBone {boneName=root}
   }
   Save cached pose 'Dynamics'(Pose=Component To Local(ComponentPose=@n24)) {cacheName=Dynamics}
  ]}
 EventGraph {nodes [
   Event BlueprintUpdateAnimation()
   > n1 = Sequence!(execute) -> (then_0, then_1, then_2:Exec, then_3:Exec, then_4:Exec)
   > n2 = "Gravity Override State Function (Parameter Asset) (Nativized)"(
    CharacterPartAnimInstance=Self-Reference()
    GameTimeInSeconds=GetGameTimeInSeconds()
    Parameters=&/M_MED_SolveSolid_Parameters
   )
   > Set Noncloth Gravity Override(Noncloth Gravity Override:Vector=@n2) -> (Output_Get:Vector)
   Branch(execute=@n1.then_1, Condition=Get bClothEnabled())
   > "Cloth Controller Function (Parameter Asset) (Nativized)"!(
    execute
    CharacterPartAnimInstance:Custom Character Part Anim Instance Object Reference=Self-Reference()
    ClothingAssetName:Name=M_MED_SolveSolid_Clothing
    ClothCoreSettings:"Cloth Core Settings Structure (by ref)"=Get Cloth Core Settings() -> (Cloth Core Settings:Cloth Core Settings Structure)
    WindGustLocalSettings:"Wind Gust Settings Structure (by ref)"=Get Wind Gust Local Settings() -> (Wind Gust Local Settings:Wind Gust Settings Structure)
    ClothUpAxis:Vector="1,0,0"
    LinVelScaleFwdBwdClamp:float=0
    LinVelScaleUpDwnClamp:float=0.7
    DeltaTime:float=Get DeltaTimeLocal() -> (DeltaTimeLocal:double)
    GameTimeInSeconds:float=GetGameTimeInSeconds()
    StateLogicLayerName:Name=none
    Parameters:Fort Character Dynamics Parameters Object Reference=&M_MED_SolveSolid_Parameters
   ) -> (
    then
    OutCurrentState:Name=None
    OutClothParameters:Cloth Parameters Structure
    OutClothCoreSettings:Cloth Core Settings Structure
    OutWindGustSettings:Wind Gust Parameters Structure
    OutWindGustLocalSettings:Wind Gust Settings Structure
    OutWindVelocity:Vector
    OutLinearVelVector:Vector
   )
   n3 = Get bClothEnabled()
   n4 = Branch(execute=@n1.then_2, Condition=AND Boolean(A=NOT Boolean(A=@n3), B=Get bRigidBodyEnabled()))
   > Set CapeFIx(CapeFIx:bool=true) -> (Output_Get:bool)
   Set CapeFIx(execute=@n4.else, CapeFIx=false)
   n5 = Branch(execute=@n1.then_3, Condition=Get IsWearingBackbling())
   n6 = GetPartSkeletalMeshComponent(self=Self-Reference())
   > Set ClothBlendWeight!(execute, ClothBlendWeight:float=0, self:Skeletal Mesh Component Object Reference=@n6) -> (then, Output_Get:float)
   > Set bDisableClothSimulation!(execute, bDisableClothSimulation:bool=true, self:Skeletal Mesh Component Object Reference=@n6) -> (then, Output_Get:bool)
   Set ClothBlendWeight(execute=@n5.else, ClothBlendWeight=1, self=@n6)
   > Set bDisableClothSimulation(bDisableClothSimulation=false, self=@n6)
   n7 = GetAnimInstance(self=GetOwnerSkeletalMeshComponent())
   n8 = Montage_IsPlaying(self=@n7, Montage=/BRCosmetics/Animation/Game/MainPlayer/Emotes/GoldenGlory/CMM/Emote_GoldenGlory_CMM_M)
   n9 = Montage_IsPlaying(self=@n7, Montage=/BRCosmetics/Animation/Game/MainPlayer/Emotes/GoldenGlory/CMM/Emote_GoldenGlory_CMM_FrontEnd_M)
   Set GoldenGloryEmote(execute=@n1.then_4, GoldenGloryEmote:bool=AND Boolean(A=OR Boolean(A=@n8, B=@n9), B=Get bRigidBodyEnabled())) -> (Output_Get:bool)
   Event OnCharacterCustomizationLoaded()
   n10 = CheckForBackbling()
   > Set IsWearingBackbling(IsWearingBackbling:bool=OR Boolean!(A=@n10, B=false, C:bool=@n10.bIsBackblingLong) -> (ReturnValue)) -> (Output_Get:bool)
  ]}
]
variables [
 CapeFIx {type=boolean}
 Cloth Core Settings {
  type=ClothCoreSettings
  default {
   collisionThickness=0
   dampingCoefficient=0.05
   frictionCoefficient=0.3
   localDampingCoefficient=0
   numberIterations=2
   numberSubsteps=1
   selfCollisionThickness=0
   animDriveDamping=<0 1>
   animDriveStiffness=<0 1>
   areaStiffness=<1 1>
   bendingStiffness=<1 1>
   edgeStiffness=<1 1>
   tetherStiffness=<1 1>
  }
 }
 ClothDeactivate {type=boolean}
 DeltaTimeLocal {type=number}
 GoldenGloryEmote {type=boolean}
 IsWearingBackbling {type=boolean}
 Noncloth Gravity Override {type=Vector}
 Wind Gust Local Settings {type=WindGustSettings, default {elapsedTime=0, finalGustStrengthLocal=0, gustDuration=0, gustStrength=0, timeBetweenGusts=0}}
]
