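# Physics asset for the M_MED_SolveSolid medium male player body, named as the collision volume for its cloth simulation.
# Capsules cover only the pelvis, the five spine bones and both upper arms, with the right arm capsule shorter and thicker than the left.
# Bodies for the root, clavicles, forearms, hands, legs, feet and dynamic front belt bones hold no shapes.
# It has no constraints between bodies.
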PhysicsAsset /BRCosmetics/Characters/Player/Male/Medium/Bodies/M_MED_SolveSolid/Meshes/M_MED_SolveSolid_ClothCollision
bodies {
 calf_l []
 calf_r []
 clavicle_l []
 clavicle_r []
 dyn_belt_front_00_rt []
 dyn_belt_front_01_rt []
 dyn_belt_front_02_rt []
 foot_l []
 foot_r []
 hand_l []
 hand_r []
 lowerarm_l []
 lowerarm_r []
 pelvis [{shapeType=Capsule, length=16, radius=14.8, center=<-2.8242 -2.3149 0>, rotation=<rot 0 -0.0006 0>}]
 root []
 spine_01 [{shapeType=Capsule, length=16.5, radius=14, center=<-2.5 -2.9459 0>}]
 spine_02 [{shapeType=Capsule, length=27, radius=10, center=<0.3095 -4.3049 0>}]
 spine_03 [{shapeType=Capsule, length=22, radius=12, shapeName=spine_03_capsule, center=<0 -3.3352 0>}]
 spine_04 [{shapeType=Capsule, length=26, radius=12, shapeName=spine_04_capsule, center=<0 -4.2016 0>}]
 spine_05 [{shapeType=Capsule, length=14, radius=15.5, shapeName=spine_05_capsule, center=<1.3653 -0.4684 0>}]
 thigh_l []
 thigh_r []
 upperarm_l [{
   shapeType=Capsule
   length=23.8269
   radius=7.2821
   shapeName=upperarm_l_capsule
   center=<9.6493 -1.2061 -1.4977>
   rotation=<rot 18.5709 -84.7781 83.4728>
  }]
 upperarm_r [{
   shapeType=Capsule
   length=16.3269
   radius=11.7821
   shapeName=upperarm_r_capsule
   center=<-12.2822 0.2621 -2.1141>
   rotation=<rot 17.7007 -84.7534 -86.5574>
  }]
}
constraints []
